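# Physics asset for the AkimboEnvoy female medium player body, covering only the torso, thighs and two jacket front bones, apparently for jacket secondary motion.
# Spheres on the spine and capsules on pelvis and thighs form a simplified body volume, with one capsule per jacket bone.
# Spine constraints chain to the pelvis with free swing and twist.
# Thighs attach to the pelvis, and jacket bones to spine_03 with narrower, uneven angular limits.
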PhysicsAsset /BRCosmetics/Characters/Player/Female/Medium/Bodies/F_MED_AkimboEnvoy/Meshes/F_MED_AkimboEnvoy_Physics
bodies {
 dyn_jacket_front_side_left_1 [{
   shapeType=Capsule
   length=4.2614
   radius=4
   shapeName=dyn_jacket_front_side_left_1_capsule
   center=<4.3524 -0.2463 2.1642>
   rotation=<rot 0.001 0.001 80>
  }]
 dyn_jacket_front_side_right_1 [{
   shapeType=Capsule
   length=6.283
   radius=4
   shapeName=dyn_jacket_front_side_right_1_capsule
   center=<5.0775 -0.0001 2.4585>
   rotation=<rot 0.0008 -0.001 90>
  }]
 pelvis [{shapeType=Capsule, length=6.4392, radius=11.1183, shapeName=pelvis_capsule, center=<0 -2.1874 0>}]
 spine_01 [{shapeType=Sphere, radius=11, shapeName=spine_01_sphere}]
 spine_02 [{shapeType=Sphere, radius=11, shapeName=spine_02_sphere}]
 spine_03 [{shapeType=Sphere, radius=10, shapeName=spine_03_sphere}]
 thigh_l [{shapeType=Capsule, length=18, radius=8.5, shapeName=thigh_l_capsule, center=<-43.177 -0.2063 -2.2027>, rotation=<rot 85 180 0>}]
 thigh_r [{shapeType=Capsule, length=18, radius=10, shapeName=thigh_r_capsule, center=<31.3729 1.1654 1.17>, rotation=<rot 85 180 180>}]
}
constraints [
 {bone1Name=thigh_l, bone2Name=pelvis, swing1LimitDegrees=45, swing2LimitDegrees=45, twistLimitDegrees=45}
 {bone1Name=thigh_r, bone2Name=pelvis, swing1LimitDegrees=45, swing2LimitDegrees=45, twistLimitDegrees=45}
 {bone1Name=dyn_jacket_front_side_left_1, bone2Name=spine_03, swing1LimitDegrees=10, swing2LimitDegrees=40, twistLimitDegrees=10}
 {bone1Name=dyn_jacket_front_side_right_1, bone2Name=spine_03, swing1LimitDegrees=10, swing2LimitDegrees=40, twistLimitDegrees=10}
 {
  bone1Name=spine_01
  bone2Name=pelvis
  swing1LimitDegrees=45
  swing1Motion=Free
  swing2LimitDegrees=45
  swing2Motion=Free
  twistLimitDegrees=45
  twistMotion=Free
 }
 {
  bone1Name=spine_02
  bone2Name=spine_01
  swing1LimitDegrees=45
  swing1Motion=Free
  swing2LimitDegrees=45
  swing2Motion=Free
  twistLimitDegrees=45
  twistMotion=Free
 }
 {
  bone1Name=spine_03
  bone2Name=spine_02
  swing1LimitDegrees=45
  swing1Motion=Free
  swing2LimitDegrees=45
  swing2Motion=Free
  twistLimitDegrees=45
  twistMotion=Free
 }
]
