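# Animation Blueprint for the F_MED_AkimboEnvoy cosmetic body, adding secondary motion and corrective posing to jacket, collar, belt and sidebag.
# It copies the parent mesh pose, then runs Control Rig, RigidBody, bone constraints and Anim Dynamics, with some jacket simulation gated outside the front end.
# Cached poses feed staged additive bone corrections, mostly weighted by leg bend, crouch, sprint, arm raise and spine rotation alphas.
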AnimBlueprint /BRCosmetics/Characters/Player/Female/Medium/Bodies/F_MED_AkimboEnvoy/Meshes/F_MED_AkimboEnvoy_AnimBP
target_skeleton=&/F_MED_AkimboEnvoy_Skeleton
graphs [
 AnimGraph {nodes [
   "Transform (Modify) Bone - Bone: dyn_collar_back_1"!(ComponentPose, bAlphaBoolEnabled:bool=True, Translation="-2,0,0", Rotation="0,0,0", Scale="0.4,0.4,0.4") -> (Pose) {
    alphaInputType=Bool
    rotationMode=BMM_Additive
    rotationSpace=BCS_BoneSpace
    scaleMode=BMM_Additive
    translationMode=BMM_Additive
    translationSpace=BCS_BoneSpace
    alphaBoolBlend {blendInTime=0.3, blendOption=ExpInOut, blendOutTime=0.3}
    alphaScaleBiasClamp {bInterpResult=true}
    boneToModify {boneName=dyn_collar_back_1}
    scale=<0.4 0.4 0.4>
    translation {x=-2}
   }
   n1 = "Transform (Modify) Bone - Bone: dyn_jacket_front_side_left_1"!(
    ComponentPose=Local To Component(LocalPose=Use cached pose 'Post Collar Setup'())
    bAlphaBoolEnabled:bool=True
    Translation="-2,2,1.5"
    Rotation="10,0,-20"
    Scale="1,1,1"
   ) -> (Pose) {
    alphaInputType=Bool
    rotationMode=BMM_Additive
    translationMode=BMM_Additive
    alphaBoolBlend {blendInTime=0.3, blendOption=ExpInOut, blendOutTime=0.3}
    boneToModify {boneName=dyn_jacket_front_side_left_1}
    rotation {pitch=10, roll=-20}
    translation=<-2 2 1.5>
   }
   n2 = "Transform (Modify) Bone - Bone: dyn_jacket_front_side_left_1"!(ComponentPose=@n1, Alpha:float=1, Translation="0,0,0", Rotation="0,0,30", Scale="1,1,1") -> (Pose) {
    rotationMode=BMM_Additive
    translationMode=BMM_Additive
    alphaBoolBlend {blendInTime=0.3, blendOption=ExpInOut, blendOutTime=0.3}
    boneToModify {boneName=dyn_jacket_front_side_left_1}
    rotation {roll=30}
   }
   n3 = "Transform (Modify) Bone - Bone: dyn_jacket_front_side_left_1"(ComponentPose=@n2, Alpha=1, Translation="0.000,1.700,4.070", Rotation="11.978,-18.785,31.257", Scale="1,1,1") {
    rotationMode=BMM_Additive
    translationMode=BMM_Additive
    alphaBoolBlend {blendInTime=0.3, blendOption=ExpInOut, blendOutTime=0.3}
    boneToModify {boneName=dyn_jacket_front_side_left_1}
    rotation=<rot 11.978 -18.785 31.257>
    translation {y=1.7, z=4.07}
   }
   n4 = "Transform (Modify) Bone - Bone: dyn_jacket_front_side_left_1"(ComponentPose=@n3, Alpha=1, Translation="0,1.7,2.5", Rotation="11.978,-18.785,31.257", Scale="1,1,1") {
    rotationMode=BMM_Additive
    translationMode=BMM_Additive
    alphaBoolBlend {blendInTime=0.3, blendOption=ExpInOut, blendOutTime=0.3}
    alphaScaleBiasClamp {bInterpResult=true}
    boneToModify {boneName=dyn_jacket_front_side_left_1}
    rotation=<rot 11.978 -18.785 31.257>
    translation {y=1.7, z=2.5}
   }
   n5 = "Transform (Modify) Bone - Bone: dyn_jacket_front_side_right_1"(ComponentPose=@n4, Alpha=1, Translation="0.000,0.000,3.655", Rotation="-18.747,23.994,-13.859", Scale="1,1,1") {
    rotationMode=BMM_Additive
    translationMode=BMM_Additive
    alphaBoolBlend {blendInTime=0.3, blendOption=ExpInOut, blendOutTime=0.3}
    boneToModify {boneName=dyn_jacket_front_side_right_1}
    rotation=<rot -18.747 23.994 -13.859>
    translation {z=3.655}
   }
   Output Pose(Result=Component To Local(ComponentPose=@n5)) {layerGroup=None}
   n6 = Anim Dynamics - Bone: dyn_jacket_zipper_01(ComponentPose=Local To Component(LocalPose=Use cached pose 'Post Rigid Body'()), Alpha=1, ExternalForce="0,0,0") {
    angularBiasOverride=1
    angularDampingOverride=1
    angularSpringConstant=10000
    bAngularSpring=true
    bOverrideAngularBias=true
    bUseGravityOverride=true
    gravityScale=0.35
    simulationSpace=BoneRelative
    alphaBoolBlend {blendOption=Linear}
    boundBone {boneName=dyn_jacket_zipper_01}
    componentAppliedLinearAccClamp=<2000 2000 2000>
    componentLinearAccScale=<0.1 0.1 0.1>
    componentLinearVelScale=<0.2 0.2 0.2>
    gravityOverride {z=-550}
    physicsBodyDefinitions [{
      boundBone {boneName=dyn_jacket_zipper_01}
      boxExtents=<10 10 10>
      constraintSetup {angularLimitsMax=<0 10 10>, angularLimitsMin=<0 -40 -10>, angularTarget=<0 0 0>}
      localJointOffset=<5 0 0>
     }]
    planarLimits [{drivingBone {boneName=thigh_r}, planeTransform {location=<0 -6 0>, rotation=<rot 0 0 -90>}}]
    relativeSpaceBone {boneName=root}
   }
   n7 = Anim Dynamics - Bone: dyn_sidebag_lower_zipper_1(ComponentPose=@n6, Alpha=1, ExternalForce="0,0,0") {
    angularBiasOverride=0.999
    angularDampingOverride=0.99
    angularSpringConstant=10000
    bOverrideAngularBias=true
    bOverrideAngularDamping=true
    bUseGravityOverride=true
    simulationSpace=BoneRelative
    alphaBoolBlend {blendOption=Linear}
    boundBone {boneName=dyn_sidebag_lower_zipper_1}
    componentAppliedLinearAccClamp=<2000 2000 2000>
    componentLinearAccScale=<0.1 0.1 0.1>
    componentLinearVelScale=<0.2 0.2 0.2>
    gravityOverride {z=-980}
    physicsBodyDefinitions [{
      boundBone {boneName=dyn_sidebag_lower_zipper_1}
      boxExtents=<10 10 10>
      constraintSetup {angularLimitsMax=<0 0 0>, angularLimitsMin=<0 -40 0>, angularTarget=<0 0 0>}
      localJointOffset=<4 0 0>
     }]
    relativeSpaceBone {boneName=root}
   }
   n8 = Anim Dynamics - Bone: dyn_sidebag_upper_strap_1(ComponentPose=@n7, Alpha=1, ExternalForce="0,0,0") {
    angularBiasOverride=0.999
    angularDampingOverride=1
    angularSpringConstant=10000
    bOverrideAngularBias=true
    bUseGravityOverride=true
    simulationSpace=BoneRelative
    alphaBoolBlend {blendOption=Linear}
    boundBone {boneName=dyn_sidebag_upper_strap_1}
    componentAppliedLinearAccClamp=<2000 2000 2000>
    componentLinearAccScale=<0.1 0.1 0.1>
    componentLinearVelScale=<0.2 0.2 0.2>
    gravityOverride {z=-500}
    physicsBodyDefinitions [{
      boundBone {boneName=dyn_sidebag_upper_strap_1}
      boxExtents=<10 10 10>
      constraintSetup {angularLimitsMax=<0 0 0>, angularLimitsMin=<0 -40 0>, angularTarget=<0 0 0>}
      localJointOffset=<4 0 0>
     }]
    relativeSpaceBone {boneName=root}
   }
   n9 = Get Not in Front End() -> (Not in Front End:bool)
   n10 = Anim Dynamics - Bone: dyn_jacket_side_left_1!(ComponentPose=@n8, bAlphaBoolEnabled:bool=@n9, ExternalForce="0,0,0") -> (Pose) {
    alphaInputType=Bool
    angularBiasOverride=0.9999
    angularDampingOverride=1
    angularSpringConstant=1000
    bOverrideAngularBias=true
    bOverrideAngularDamping=true
    bUseGravityOverride=true
    simulationSpace=BoneRelative
    alphaBoolBlend {blendInTime=0.1, blendOption=Linear, blendOutTime=0.1}
    boundBone {boneName=dyn_jacket_side_left_1}
    componentAppliedLinearAccClamp=<2000 2000 2000>
    componentLinearAccScale=<0.6 0.6 0.6>
    componentLinearVelScale=<0.6 0.6 0.6>
    gravityOverride {z=-980}
    physicsBodyDefinitions [{
      boundBone {boneName=dyn_jacket_side_left_1}
      boxExtents=<20 20 20>
      constraintSetup {angularLimitsMax=<0 15 2>, angularLimitsMin=<0 -40 -2>, angularTarget=<0 0 0>}
      localJointOffset=<-10 0 -5>
     }]
    relativeSpaceBone {boneName=spine_03}
   }
   n11 = Anim Dynamics - Bone: dyn_jacket_back_side_left_1!(ComponentPose=@n10, bAlphaBoolEnabled:bool=@n9, ExternalForce="0,0,0") -> (Pose) {
    alphaInputType=Bool
    angularBiasOverride=0.9999
    angularDampingOverride=1
    angularSpringConstant=1000
    bOverrideAngularBias=true
    bOverrideAngularDamping=true
    bUseGravityOverride=true
    simulationSpace=BoneRelative
    alphaBoolBlend {blendInTime=0.1, blendOption=Linear, blendOutTime=0.1}
    boundBone {boneName=dyn_jacket_back_side_left_1}
    componentAppliedLinearAccClamp=<2000 2000 2000>
    componentLinearAccScale=<0.6 0.6 0.6>
    componentLinearVelScale=<0.6 0.6 0.6>
    gravityOverride {z=-980}
    physicsBodyDefinitions [{
      boundBone {boneName=dyn_jacket_back_side_left_1}
      boxExtents=<20 20 20>
      constraintSetup {angularLimitsMax=<0 15 2>, angularLimitsMin=<0 -40 -2>, angularTarget=<0 0 0>}
      localJointOffset=<-10 0 -5>
     }]
    relativeSpaceBone {boneName=spine_03}
   }
   n12 = Anim Dynamics - Bone: dyn_jacket_back_1!(ComponentPose=@n11, bAlphaBoolEnabled:bool=@n9, ExternalForce="0,0,0") -> (Pose) {
    alphaInputType=Bool
    angularBiasOverride=0.9999
    angularDampingOverride=1
    angularSpringConstant=1000
    bOverrideAngularBias=true
    bOverrideAngularDamping=true
    bUseGravityOverride=true
    simulationSpace=BoneRelative
    alphaBoolBlend {blendInTime=0.1, blendOption=Linear, blendOutTime=0.1}
    boundBone {boneName=dyn_jacket_back_1}
    componentAppliedLinearAccClamp=<2000 2000 2000>
    componentLinearAccScale=<0.6 0.6 0.6>
    componentLinearVelScale=<0.6 0.6 0.6>
    gravityOverride {z=-980}
    physicsBodyDefinitions [{
      boundBone {boneName=dyn_jacket_back_1}
      boxExtents=<20 20 20>
      constraintSetup {angularLimitsMax=<0 15 2>, angularLimitsMin=<0 -40 -2>, angularTarget=<0 0 0>}
      localJointOffset=<-10 0 -5>
     }]
    relativeSpaceBone {boneName=spine_03}
   }
   n13 = Anim Dynamics - Bone: dyn_jacket_back_side_right_1!(ComponentPose=@n12, bAlphaBoolEnabled:bool=@n9, ExternalForce="0,0,0") -> (Pose) {
    alphaInputType=Bool
    angularBiasOverride=0.9999
    angularDampingOverride=1
    angularSpringConstant=10000
    bOverrideAngularBias=true
    bOverrideAngularDamping=true
    bUseGravityOverride=true
    simulationSpace=BoneRelative
    alphaBoolBlend {blendInTime=0.1, blendOption=Linear, blendOutTime=0.1}
    boundBone {boneName=dyn_jacket_back_side_right_1}
    componentAppliedLinearAccClamp=<2000 2000 2000>
    componentLinearAccScale=<0.6 0.6 0.6>
    componentLinearVelScale=<0.6 0.6 0.6>
    gravityOverride {z=-980}
    physicsBodyDefinitions [{
      boundBone {boneName=dyn_jacket_back_side_right_1}
      boxExtents=<20 20 20>
      constraintSetup {angularLimitsMax=<0 15 2>, angularLimitsMin=<0 -40 -2>, angularTarget=<0 0 0>}
      localJointOffset=<-10 0 -5>
     }]
    relativeSpaceBone {boneName=spine_03}
   }
   n14 = Anim Dynamics - Bone: dyn_jacket__side_right_1!(ComponentPose=@n13, bAlphaBoolEnabled:bool=@n9, ExternalForce="0,0,0") -> (Pose) {
    alphaInputType=Bool
    angularBiasOverride=0.9999
    angularDampingOverride=1
    angularSpringConstant=10000
    bOverrideAngularBias=true
    bOverrideAngularDamping=true
    bUseGravityOverride=true
    simulationSpace=BoneRelative
    alphaBoolBlend {blendInTime=0.1, blendOption=Linear, blendOutTime=0.1}
    boundBone {boneName=dyn_jacket__side_right_1}
    componentAppliedLinearAccClamp=<2000 2000 2000>
    componentLinearAccScale=<0.6 0.6 0.6>
    componentLinearVelScale=<0.6 0.6 0.6>
    gravityOverride {z=-980}
    physicsBodyDefinitions [{
      boundBone {boneName=dyn_jacket__side_right_1}
      boxExtents=<20 20 20>
      constraintSetup {angularLimitsMax=<0 15 2>, angularLimitsMin=<0 -40 -2>, angularTarget=<0 0 0>}
      localJointOffset=<-10 0 -5>
     }]
    relativeSpaceBone {boneName=spine_03}
   }
   n15 = Anim Dynamics - Bone: dyn_collar_front_1(ComponentPose=@n14, Alpha=1, ExternalForce="0,0,0") {
    angularBiasOverride=0.999
    angularDampingOverride=0.99
    angularSpringConstant=10000
    bOverrideAngularBias=true
    bOverrideAngularDamping=true
    bUseGravityOverride=true
    simulationSpace=BoneRelative
    alphaBoolBlend {blendOption=Linear}
    boundBone {boneName=dyn_collar_front_1}
    componentAppliedLinearAccClamp=<1000 1000 1000>
    componentLinearAccScale=<0.2 0.2 0.2>
    componentLinearVelScale=<0.2 0.2 0.2>
    gravityOverride {z=400}
    physicsBodyDefinitions [{
      boundBone {boneName=dyn_collar_front_1}
      boxExtents=<10 15 15>
      constraintSetup {angularLimitsMax=<0 -5 -5>, angularLimitsMin=<0 -20 -5>, angularTarget=<0 0 0>}
      localJointOffset=<12 0 0>
     }]
    relativeSpaceBone {boneName=spine_05}
   }
   n16 = Anim Dynamics - Bone: dyn_collar_front_right_1(ComponentPose=@n15, Alpha=1, ExternalForce="0,0,0") {
    angularBiasOverride=0.999
    angularDampingOverride=0.99
    angularSpringConstant=10000
    bOverrideAngularBias=true
    bOverrideAngularDamping=true
    bUseGravityOverride=true
    simulationSpace=BoneRelative
    alphaBoolBlend {blendOption=Linear}
    boundBone {boneName=dyn_collar_front_right_1}
    componentAppliedLinearAccClamp=<1000 1000 1000>
    componentLinearAccScale=<0.2 0.2 0.2>
    componentLinearVelScale=<0.2 0.2 0.2>
    gravityOverride {z=400}
    physicsBodyDefinitions [{
      boundBone {boneName=dyn_collar_front_right_1}
      boxExtents=<10 15 15>
      constraintSetup {angularLimitsMax=<0 -5 -5>, angularLimitsMin=<0 -20 -5>, angularTarget=<0 0 0>}
      localJointOffset=<12 0 0>
     }]
    relativeSpaceBone {boneName=spine_05}
   }
   n17 = Anim Dynamics - Bone: dyn_collar_back_right_1(ComponentPose=@n16, Alpha=1, ExternalForce="0,0,0") {
    angularBiasOverride=0.999
    angularDampingOverride=0.99
    angularSpringConstant=10000
    bOverrideAngularBias=true
    bOverrideAngularDamping=true
    bUseGravityOverride=true
    simulationSpace=BoneRelative
    alphaBoolBlend {blendOption=Linear}
    boundBone {boneName=dyn_collar_back_right_1}
    componentAppliedLinearAccClamp=<1000 1000 1000>
    componentLinearAccScale=<0.2 0.2 0.2>
    componentLinearVelScale=<0.2 0.2 0.2>
    gravityOverride {z=400}
    physicsBodyDefinitions [{
      boundBone {boneName=dyn_collar_back_right_1}
      boxExtents=<10 15 15>
      constraintSetup {angularLimitsMax=<0 -5 -5>, angularLimitsMin=<0 -20 -5>, angularTarget=<0 0 0>}
      localJointOffset=<12 0 0>
     }]
    relativeSpaceBone {boneName=spine_05}
   }
   n18 = Anim Dynamics - Bone: dyn_collar_back_1(ComponentPose=@n17, Alpha=1, ExternalForce="0,0,0") {
    angularBiasOverride=0.999
    angularDampingOverride=0.99
    angularSpringConstant=10000
    bOverrideAngularBias=true
    bOverrideAngularDamping=true
    bUseGravityOverride=true
    simulationSpace=BoneRelative
    alphaBoolBlend {blendOption=Linear}
    boundBone {boneName=dyn_collar_back_1}
    componentAppliedLinearAccClamp=<1000 1000 1000>
    componentLinearAccScale=<0.2 0.2 0.2>
    componentLinearVelScale=<0.2 0.2 0.2>
    gravityOverride {z=400}
    physicsBodyDefinitions [{
      boundBone {boneName=dyn_collar_back_1}
      boxExtents=<10 15 15>
      constraintSetup {angularLimitsMax=<0 -5 -5>, angularLimitsMin=<0 -20 -5>, angularTarget=<0 0 0>}
      localJointOffset=<12 0 0>
     }]
    relativeSpaceBone {boneName=spine_05}
   }
   n19 = Anim Dynamics - Bone: dyn_collar_back_left_1(ComponentPose=@n18, Alpha=1, ExternalForce="0,0,0") {
    angularBiasOverride=0.999
    angularDampingOverride=0.99
    angularSpringConstant=10000
    bOverrideAngularBias=true
    bOverrideAngularDamping=true
    bUseGravityOverride=true
    simulationSpace=BoneRelative
    alphaBoolBlend {blendOption=Linear}
    boundBone {boneName=dyn_collar_back_left_1}
    componentAppliedLinearAccClamp=<1000 1000 1000>
    componentLinearAccScale=<0.2 0.2 0.2>
    componentLinearVelScale=<0.2 0.2 0.2>
    gravityOverride {z=400}
    physicsBodyDefinitions [{
      boundBone {boneName=dyn_collar_back_left_1}
      boxExtents=<10 15 15>
      constraintSetup {angularLimitsMax=<0 -5 -5>, angularLimitsMin=<0 -20 -5>, angularTarget=<0 0 0>}
      localJointOffset=<12 0 0>
     }]
    relativeSpaceBone {boneName=spine_05}
   }
   n20 = Anim Dynamics - Bone: dyn_collar_front_left_1(ComponentPose=@n19, Alpha=1, ExternalForce="0,0,0") {
    angularBiasOverride=0.999
    angularDampingOverride=0.99
    angularSpringConstant=10000
    bOverrideAngularBias=true
    bOverrideAngularDamping=true
    bUseGravityOverride=true
    simulationSpace=BoneRelative
    alphaBoolBlend {blendOption=Linear}
    boundBone {boneName=dyn_collar_front_left_1}
    componentAppliedLinearAccClamp=<1000 1000 1000>
    componentLinearAccScale=<0.2 0.2 0.2>
    componentLinearVelScale=<0.2 0.2 0.2>
    gravityOverride {z=400}
    physicsBodyDefinitions [{
      boundBone {boneName=dyn_collar_front_left_1}
      boxExtents=<10 15 15>
      constraintSetup {angularLimitsMax=<0 -5 -5>, angularLimitsMin=<0 -20 -5>, angularTarget=<0 0 0>}
      localJointOffset=<12 0 0>
     }]
    relativeSpaceBone {boneName=spine_05}
   }
   Save cached pose 'Post-Anim Dynamic'(Pose=Component To Local(ComponentPose=@n20)) {cacheName=Post-Anim Dynamic}
   n21 = Control Rig(Source=Copy Pose From Mesh() {bUseAttachedParent=true}) {
    alphaBoolBlend {blendOption=Linear}
    controlRigAssetReference {blueprintRigClass=&/F_MED_AkimboEnvoy_CtrlRig.F_MED_AkimboEnvoy_CtrlRig_C}
   }
   Save cached pose 'Post Control Rig'(Pose=@n21) {cacheName=Post Control Rig}
   n22 = RigidBody(ComponentPose=Local To Component(LocalPose=Use cached pose 'Post Control Rig'())) {
    bOverrideWorldGravity=true
    overridePhysicsAsset=&/F_MED_AkimboEnvoy_Physics
    simulationSpace=BaseBoneSpace
    alphaBoolBlend {blendOption=Linear}
    baseBoneRef {boneName=pelvis}
    componentAppliedLinearAccClamp=<2000 2000 2000>
    componentLinearAccScale=<0.6 0.6 0.6>
    componentLinearVelScale=<0.6 0.6 0.6>
    overrideWorldGravity {z=-980}
   }
   n23 = "Constraint - dyn_jacket_zipper_MSTR (P)"(ComponentPose=@n22, Alpha=1, ConstraintWeights_0=1) {
    alphaBoolBlend {blendOption=Linear}
    boneToModify {boneName=dyn_jacket_zipper_MSTR}
    constraintSetup [{targetBone {boneName=dyn_jacket_front_side_right_1}}]
    constraintWeights [1]
   }
   n24 = "Constraint - dyn_sidebag_lower_zipper_MSTR (PPP)"!(ComponentPose=@n23, Alpha=1, ConstraintWeights_0=0.123, ConstraintWeights_1:float=0.027, ConstraintWeights_2:float=0.85) -> (Pose) {
    alphaBoolBlend {blendOption=Linear}
    boneToModify {boneName=dyn_sidebag_lower_zipper_MSTR}
    constraintSetup [{targetBone {boneName=deform_lower_belt_back_right}}, {targetBone {boneName=pelvis}}, {targetBone {boneName=thigh_twist_01_r}}]
    constraintWeights [1,1,1]
   }
   Save cached pose 'Post Rigid Body'(Pose=Component To Local(ComponentPose=@n24)) {cacheName=Post Rigid Body}
   n25 = Get Left Leg Bend to the Left Alpha() -> (Left Leg Bend to the Left Alpha:double)
   n26 = Get Left Leg crouch moving Bend Euler Z Alpha() -> (Left Leg crouch moving Bend Euler Z Alpha:double)
   n27 = Get Right Leg Bend Euler Z Alpha() -> (Right Leg Bend Euler Z Alpha:double)
   n28 = "Transform (Modify) Bone - Bone: deform_lower_belt_back_right"(
    ComponentPose=Local To Component(LocalPose=Use cached pose 'Post-Anim Dynamic'())
    Alpha=@n27
    Translation="0,0,-1.5"
    Rotation="0,10,0"
    Scale="1,1,1"
   ) {
    rotationMode=BMM_Additive
    rotationSpace=BCS_BoneSpace
    translationMode=BMM_Additive
    translationSpace=BCS_BoneSpace
    alphaBoolBlend {blendOption=Linear}
    boneToModify {boneName=deform_lower_belt_back_right}
    rotation {yaw=10}
    translation {z=-1.5}
   }
   n29 = "Transform (Modify) Bone - Bone: deform_lower_belt_front_right"(ComponentPose=@n28, Alpha=@n27, Translation="-7,-2,5", Rotation="90,0,0", Scale="1,1,1") {
    rotationMode=BMM_Additive
    rotationSpace=BCS_BoneSpace
    translationMode=BMM_Additive
    translationSpace=BCS_BoneSpace
    alphaBoolBlend {blendOption=Linear}
    boneToModify {boneName=deform_lower_belt_front_right}
    rotation {pitch=90}
    translation=<-7 -2 5>
   }
   n30 = Get Left Leg Bend Euler Z Alpha() -> (Left Leg Bend Euler Z Alpha:double)
   n31 = "Transform (Modify) Bone - Bone: deform_lower_belt_front_left"(ComponentPose=@n29, Alpha=@n30, Translation="-2,0,5", Rotation="60,0,0", Scale="1,1,1") {
    rotationMode=BMM_Additive
    rotationSpace=BCS_BoneSpace
    translationMode=BMM_Additive
    translationSpace=BCS_BoneSpace
    alphaBoolBlend {blendOption=Linear}
    boneToModify {boneName=deform_lower_belt_front_left}
    rotation {pitch=60}
    translation {x=-2, z=5}
   }
   n32 = "Transform (Modify) Bone - Bone: deform_lower_belt_left"(ComponentPose=@n31, Alpha=@n30, Translation="-2,2,2", Rotation="0,0,0", Scale="1,1,1") {
    rotationMode=BMM_Additive
    rotationSpace=BCS_BoneSpace
    translationMode=BMM_Additive
    translationSpace=BCS_BoneSpace
    alphaBoolBlend {blendOption=Linear}
    boneToModify {boneName=deform_lower_belt_left}
    translation=<-2 2 2>
   }
   n33 = "Transform (Modify) Bone - Bone: deform_lower_belt_front_right"(
    ComponentPose=@n32
    Alpha=Get Right Leg crouch Bend Euler Z Alpha() -> (Right Leg crouch Bend Euler Z Alpha:double)
    Translation="-0.872,-7.269,12.978"
    Rotation="35.075,-1.663,-39.509"
    Scale="1,1,1"
   ) {
    rotationMode=BMM_Additive
    rotationSpace=BCS_BoneSpace
    translationMode=BMM_Additive
    translationSpace=BCS_BoneSpace
    alphaBoolBlend {blendOption=Linear}
    boneToModify {boneName=deform_lower_belt_front_right}
    rotation=<rot 35.0751 -1.663 -39.509>
    translation=<-0.872 -7.269 12.978>
   }
   n34 = Get Melee Spine rotation to the Left Alpha() -> (Melee Spine rotation to the Left Alpha:float)
   n35 = "Transform (Modify) Bone - Bone: deform_lower_belt_front_left"(ComponentPose=@n33, Alpha=@n34, Translation="1,10,5", Rotation="0,0,0", Scale="1,1,1") {
    rotationMode=BMM_Additive
    rotationSpace=BCS_BoneSpace
    translationMode=BMM_Additive
    translationSpace=BCS_BoneSpace
    alphaBoolBlend {blendOption=Linear}
    alphaScaleBiasClamp {interpSpeedIncreasing=1}
    boneToModify {boneName=deform_lower_belt_front_left}
    translation=<1 10 5>
   }
   n36 = "Transform (Modify) Bone - Bone: deform_lower_belt_left"(ComponentPose=@n35, Alpha=@n34, Translation="4,10,0", Rotation="0,0,0", Scale="1,1,1") {
    rotationMode=BMM_Additive
    rotationSpace=BCS_BoneSpace
    translationMode=BMM_Additive
    translationSpace=BCS_BoneSpace
    alphaBoolBlend {blendOption=Linear}
    alphaScaleBiasClamp {interpSpeedIncreasing=1}
    boneToModify {boneName=deform_lower_belt_left}
    translation=<4 10>
   }
   n37 = "Transform (Modify) Bone - Bone: deform_lower_belt_back_left"(ComponentPose=@n36, Alpha=@n34, Translation="0,0,0", Rotation="0,0,0", Scale="1,1,1") {
    rotationMode=BMM_Additive
    rotationSpace=BCS_BoneSpace
    translationMode=BMM_Additive
    translationSpace=BCS_BoneSpace
    alphaBoolBlend {blendOption=Linear}
    alphaScaleBiasClamp {interpSpeedIncreasing=1}
    boneToModify {boneName=deform_lower_belt_back_left}
   }
   n38 = Get Melee Spine rotation to the Right Alpha() -> (Melee Spine rotation to the Right Alpha:float)
   n39 = "Transform (Modify) Bone - Bone: deform_lower_belt_front_left"(ComponentPose=@n37, Alpha=@n38, Translation="0,0,0", Rotation="0,0,0", Scale="1,1,1") {
    rotationMode=BMM_Additive
    rotationSpace=BCS_BoneSpace
    translationMode=BMM_Additive
    translationSpace=BCS_BoneSpace
    alphaBoolBlend {blendOption=Linear}
    alphaScaleBiasClamp {interpSpeedIncreasing=1}
    boneToModify {boneName=deform_lower_belt_front_left}
   }
   n40 = "Transform (Modify) Bone - Bone: deform_lower_belt_left"(ComponentPose=@n39, Alpha=@n38, Translation="0,0,0", Rotation="0,0,0", Scale="1,1,1") {
    rotationMode=BMM_Additive
    rotationSpace=BCS_BoneSpace
    translationMode=BMM_Additive
    translationSpace=BCS_BoneSpace
    alphaBoolBlend {blendOption=Linear}
    alphaScaleBiasClamp {interpSpeedIncreasing=1}
    boneToModify {boneName=deform_lower_belt_left}
   }
   n41 = "Transform (Modify) Bone - Bone: deform_lower_belt_back_left"(ComponentPose=@n40, Alpha=@n38, Translation="0,0,0", Rotation="0,0,0", Scale="1,1,1") {
    rotationMode=BMM_Additive
    rotationSpace=BCS_BoneSpace
    translationMode=BMM_Additive
    translationSpace=BCS_BoneSpace
    alphaBoolBlend {blendOption=Linear}
    alphaScaleBiasClamp {interpSpeedIncreasing=1}
    boneToModify {boneName=deform_lower_belt_back_left}
   }
   n42 = "Transform (Modify) Bone - Bone: deform_lower_belt_front_left"(ComponentPose=@n41, Alpha=@n26, Translation="-8.085,6.930,8.392", Rotation="-3.432,4.181,55.754", Scale="1,1,1") {
    rotationMode=BMM_Additive
    rotationSpace=BCS_BoneSpace
    translationMode=BMM_Additive
    translationSpace=BCS_BoneSpace
    alphaBoolBlend {blendOption=Linear}
    alphaScaleBiasClamp {bInterpResult=true, interpSpeedDecreasing=20, interpSpeedIncreasing=30}
    boneToModify {boneName=deform_lower_belt_front_left}
    rotation=<rot -3.4319 4.181 55.754>
    translation=<-8.085 6.93 8.392>
   }
   n43 = "Transform (Modify) Bone - Bone: deform_lower_belt_left"(ComponentPose=@n42, Alpha=@n26, Translation="-5.337,-2.874,-3.428", Rotation="7.745,108.277,43.479", Scale="1,1,1") {
    rotationMode=BMM_Additive
    rotationSpace=BCS_BoneSpace
    translationMode=BMM_Additive
    translationSpace=BCS_BoneSpace
    alphaBoolBlend {blendOption=Linear}
    alphaScaleBiasClamp {bInterpResult=true, interpSpeedDecreasing=20, interpSpeedIncreasing=30}
    boneToModify {boneName=deform_lower_belt_left}
    rotation=<rot 7.745 108.2771 43.479>
    translation=<-5.337 -2.874 -3.428>
   }
   n44 = "Transform (Modify) Bone - Bone: deform_lower_belt_back_left"(ComponentPose=@n43, Alpha=@n26, Translation="-5.444,-2.414,-0.763", Rotation="14.108,20.649,5.183", Scale="1,1,1") {
    rotationMode=BMM_Additive
    rotationSpace=BCS_BoneSpace
    translationMode=BMM_Additive
    translationSpace=BCS_BoneSpace
    alphaBoolBlend {blendOption=Linear}
    alphaScaleBiasClamp {bInterpResult=true, interpSpeedIncreasing=20}
    boneToModify {boneName=deform_lower_belt_back_left}
    rotation=<rot 14.108 20.6491 5.183>
    translation=<-5.444 -2.414 -0.763>
   }
   n45 = Get Left Leg crouch Idle Bend Euler Z Alpha() -> (Left Leg crouch Idle Bend Euler Z Alpha:double)
   n46 = "Transform (Modify) Bone - Bone: deform_lower_belt_front_left"(ComponentPose=@n44, Alpha=@n45, Translation="-11.280,-8.544,7.347", Rotation="51.718,-20.087,92.180", Scale="1,1,1") {
    rotationMode=BMM_Additive
    rotationSpace=BCS_BoneSpace
    translationMode=BMM_Additive
    translationSpace=BCS_BoneSpace
    alphaBoolBlend {blendOption=Linear}
    alphaScaleBiasClamp {bInterpResult=true, interpSpeedDecreasing=20, interpSpeedIncreasing=20}
    boneToModify {boneName=deform_lower_belt_front_left}
    rotation=<rot 51.7181 -20.087 92.18>
    translation=<-11.28 -8.544 7.347>
   }
   n47 = "Transform (Modify) Bone - Bone: deform_lower_belt_left"(ComponentPose=@n46, Alpha=@n45, Translation="-3.490,-4.799,-3.320", Rotation="31.617,33.185,-20.499", Scale="1,1,1") {
    rotationMode=BMM_Additive
    rotationSpace=BCS_BoneSpace
    translationMode=BMM_Additive
    translationSpace=BCS_BoneSpace
    alphaBoolBlend {blendOption=Linear}
    alphaScaleBiasClamp {bInterpResult=true, interpSpeedDecreasing=20, interpSpeedIncreasing=20}
    boneToModify {boneName=deform_lower_belt_left}
    rotation=<rot 31.617 33.1851 -20.499>
    translation=<-3.49 -4.799 -3.32>
   }
   n48 = "Transform (Modify) Bone - Bone: deform_lower_belt_back_left"(ComponentPose=@n47, Alpha=@n45, Translation="1.353,0.353,0.701", Rotation="-19.954,20.151,-1.728", Scale="1,1,1") {
    rotationMode=BMM_Additive
    rotationSpace=BCS_BoneSpace
    translationMode=BMM_Additive
    translationSpace=BCS_BoneSpace
    alphaBoolBlend {blendOption=Linear}
    alphaScaleBiasClamp {bInterpResult=true, interpSpeedDecreasing=20, interpSpeedIncreasing=20}
    boneToModify {boneName=deform_lower_belt_back_left}
    rotation=<rot -19.954 20.151 -1.728>
    translation=<1.353 0.353 0.701>
   }
   n49 = Get Left Leg Bend Standing Euler Z Alpha() -> (Left Leg Bend Standing Euler Z Alpha:double)
   n50 = "Transform (Modify) Bone - Bone: deform_lower_belt_front_left"(ComponentPose=@n48, Alpha=@n49, Translation="-2.108,0.000,-0.000", Rotation="0,0,0", Scale="1,1,1") {
    rotationMode=BMM_Additive
    rotationSpace=BCS_BoneSpace
    translationMode=BMM_Additive
    translationSpace=BCS_BoneSpace
    alphaBoolBlend {blendOption=Linear}
    boneToModify {boneName=deform_lower_belt_front_left}
    translation {x=-2.108}
   }
   n51 = "Transform (Modify) Bone - Bone: deform_lower_belt_left"(ComponentPose=@n50, Alpha=@n49, Translation="-5.862,-0.000,0.301", Rotation="0,0,0", Scale="1,1,1") {
    rotationMode=BMM_Additive
    rotationSpace=BCS_BoneSpace
    translationMode=BMM_Additive
    translationSpace=BCS_BoneSpace
    alphaBoolBlend {blendOption=Linear}
    boneToModify {boneName=deform_lower_belt_left}
    translation {x=-5.862, z=0.301}
   }
   n52 = Get Left Leg Bend Euler Z Sprint Alpha() -> (Left Leg Bend Euler Z Sprint Alpha:double)
   n53 = "Transform (Modify) Bone - Bone: deform_lower_belt_front_left"(ComponentPose=@n51, Alpha=@n52, Translation="-3.460,5.953,8.666", Rotation="20.000,-0.000,10.000", Scale="1,1,1") {
    rotationMode=BMM_Additive
    rotationSpace=BCS_BoneSpace
    translationMode=BMM_Additive
    translationSpace=BCS_BoneSpace
    alphaBoolBlend {blendOption=Linear}
    boneToModify {boneName=deform_lower_belt_front_left}
    rotation=<rot 20.0001 0 10>
    translation=<-3.4598 5.9529 8.6656>
   }
   n54 = "Transform (Modify) Bone - Bone: deform_lower_belt_left"(ComponentPose=@n53, Alpha=@n52, Translation="-9.290,0.337,-0.259", Rotation="3.405,90.593,19.721", Scale="1,1,1") {
    rotationMode=BMM_Additive
    rotationSpace=BCS_BoneSpace
    translationMode=BMM_Additive
    translationSpace=BCS_BoneSpace
    alphaBoolBlend {blendOption=Linear}
    boneToModify {boneName=deform_lower_belt_left}
    rotation=<rot 3.405 90.5931 19.721>
    translation=<-9.2901 0.337 -0.259>
   }
   n55 = "Transform (Modify) Bone - Bone: deform_lower_belt_front_left"(ComponentPose=@n54, Alpha=@n25, Translation="-0.461,1.935,1.829", Rotation="0,0,0", Scale="1,1,1") {
    rotationMode=BMM_Additive
    rotationSpace=BCS_BoneSpace
    translationMode=BMM_Additive
    translationSpace=BCS_BoneSpace
    alphaBoolBlend {blendOption=Linear}
    alphaScaleBiasClamp {interpSpeedIncreasing=1}
    boneToModify {boneName=deform_lower_belt_front_left}
    translation=<-0.461 1.935 1.829>
   }
   n56 = "Transform (Modify) Bone - Bone: deform_lower_belt_left"(ComponentPose=@n55, Alpha=@n25, Translation="-5.000,0.000,0.811", Rotation="0,0,0", Scale="1,1,1") {
    rotationMode=BMM_Additive
    rotationSpace=BCS_BoneSpace
    translationMode=BMM_Additive
    translationSpace=BCS_BoneSpace
    alphaBoolBlend {blendOption=Linear}
    alphaScaleBiasClamp {interpSpeedIncreasing=1}
    boneToModify {boneName=deform_lower_belt_left}
    translation {x=-5, z=0.811}
   }
   n57 = "Transform (Modify) Bone - Bone: deform_lower_belt_back_left"(ComponentPose=@n56, Alpha=@n25, Translation="-1,0,0", Rotation="10.000,0.000,0.000", Scale="1,1,1") {
    rotationMode=BMM_Additive
    rotationSpace=BCS_BoneSpace
    translationMode=BMM_Additive
    translationSpace=BCS_BoneSpace
    alphaBoolBlend {blendOption=Linear}
    alphaScaleBiasClamp {interpSpeedIncreasing=1}
    boneToModify {boneName=deform_lower_belt_back_left}
    rotation=<rot 9.9999 0.0001 0>
    translation {x=-1}
   }
   n58 = "Transform (Modify) Bone - Bone: deform_lower_belt_front_right"(ComponentPose=@n57, Alpha=@n25, Translation="0,2,1.5", Rotation="0,0,0", Scale="1,1,1") {
    rotationMode=BMM_Additive
    rotationSpace=BCS_BoneSpace
    translationMode=BMM_Additive
    translationSpace=BCS_BoneSpace
    alphaBoolBlend {blendOption=Linear}
    boneToModify {boneName=deform_lower_belt_front_right}
    translation {y=2, z=1.5}
   }
   n59 = Get Left Leg Bend to the Right Alpha() -> (Left Leg Bend to the Right Alpha:double)
   n60 = "Transform (Modify) Bone - Bone: deform_lower_belt_front_left"(ComponentPose=@n58, Alpha=@n59, Translation="0,9,12", Rotation="0.000,-20.000,0.000", Scale="1,1,1") {
    rotationMode=BMM_Additive
    rotationSpace=BCS_BoneSpace
    translationMode=BMM_Additive
    translationSpace=BCS_BoneSpace
    alphaBoolBlend {blendOption=Linear}
    alphaScaleBiasClamp {interpSpeedIncreasing=1}
    boneToModify {boneName=deform_lower_belt_front_left}
    rotation=<rot 0 -20.0001 0>
    translation {y=9, z=12}
   }
   n61 = "Transform (Modify) Bone - Bone: deform_lower_belt_left"(ComponentPose=@n60, Alpha=@n59, Translation="9.207,-11.148,0.238", Rotation="0.000,0.000,-40.000", Scale="1,1,1") {
    rotationMode=BMM_Additive
    rotationSpace=BCS_BoneSpace
    translationMode=BMM_Additive
    translationSpace=BCS_BoneSpace
    alphaBoolBlend {blendOption=Linear}
    alphaScaleBiasClamp {interpSpeedIncreasing=1}
    boneToModify {boneName=deform_lower_belt_left}
    rotation=<rot 0 0 -40>
    translation=<9.207 -11.148 0.238>
   }
   n62 = "Transform (Modify) Bone - Bone: deform_lower_belt_back_left"(ComponentPose=@n61, Alpha=@n59, Translation="1.588,3.301,0.396", Rotation="-20.000,0.000,-0.000", Scale="1,1,1") {
    rotationMode=BMM_Additive
    rotationSpace=BCS_BoneSpace
    translationMode=BMM_Additive
    translationSpace=BCS_BoneSpace
    alphaBoolBlend {blendOption=Linear}
    alphaScaleBiasClamp {interpSpeedIncreasing=1}
    boneToModify {boneName=deform_lower_belt_back_left}
    rotation=<rot -20 0.0001 0>
    translation=<1.588 3.301 0.396>
   }
   n63 = "Transform (Modify) Bone - Bone: deform_sidebag_01"(
    ComponentPose=@n62
    Alpha=Get Sidebag correction Alpha() -> (Sidebag correction Alpha:double)
    Translation="0.077,3.172,-0.835"
    Rotation="-9.656,17.387,15.221"
    Scale="1,1,1"
   ) {
    rotationMode=BMM_Additive
    rotationSpace=BCS_BoneSpace
    translationMode=BMM_Additive
    translationSpace=BCS_BoneSpace
    alphaBoolBlend {blendOption=Linear}
    alphaScaleBiasClamp {interpSpeedIncreasing=1}
    boneToModify {boneName=deform_sidebag_01}
    rotation=<rot -9.6561 17.3869 15.221>
    translation=<0.077 3.172 -0.835>
   }
   n64 = "Transform (Modify) Bone - Bone: deform_sidebag_01"!(ComponentPose=@n63, bAlphaBoolEnabled:bool=True, Translation="0,7.5,-3.5", Rotation="-7,7,45", Scale="1,1,1") -> (Pose) {
    alphaInputType=Bool
    rotationMode=BMM_Additive
    rotationSpace=BCS_BoneSpace
    translationMode=BMM_Additive
    translationSpace=BCS_BoneSpace
    alphaBoolBlend {blendInTime=0.3, blendOption=CircularInOut, blendOutTime=0.3}
    alphaScaleBiasClamp {interpSpeedIncreasing=1}
    boneToModify {boneName=deform_sidebag_01}
    rotation=<rot -7 7 45>
    translation {y=7.5, z=-3.5}
   }
   n65 = "Transform (Modify) Bone - Bone: deform_lower_belt_back_right"(
    ComponentPose=@n64
    Alpha=Get Right Leg crouch moving Bend Euler Z Alpha() -> (Right Leg crouch moving Bend Euler Z Alpha:double)
    Translation="-2,0,0"
    Rotation="0,-20,0"
    Scale="1,1,1"
   ) {
    rotationMode=BMM_Additive
    rotationSpace=BCS_BoneSpace
    translationMode=BMM_Additive
    translationSpace=BCS_BoneSpace
    alphaBoolBlend {blendOption=Linear}
    boneToModify {boneName=deform_lower_belt_back_right}
    rotation {yaw=-20}
    translation {x=-2}
   }
   n66 = Get Right Leg crouch Idle Bend Euler Z Alpha() -> (Right Leg crouch Idle Bend Euler Z Alpha:double)
   n67 = "Transform (Modify) Bone - Bone: deform_lower_belt_back_right"(ComponentPose=@n65, Alpha=@n66, Translation="-2,0,1", Rotation="0,-20,0", Scale="1,1,1") {
    rotationMode=BMM_Additive
    rotationSpace=BCS_BoneSpace
    translationMode=BMM_Additive
    translationSpace=BCS_BoneSpace
    alphaBoolBlend {blendOption=Linear}
    boneToModify {boneName=deform_lower_belt_back_right}
    rotation {yaw=-20}
    translation {x=-2, z=1}
   }
   n68 = "Transform (Modify) Bone - Bone: deform_lower_belt_front_right"(ComponentPose=@n67, Alpha=@n66, Translation="-3,-1,0", Rotation="0,10,0", Scale="1,1,1") {
    rotationMode=BMM_Additive
    rotationSpace=BCS_BoneSpace
    translationMode=BMM_Additive
    translationSpace=BCS_BoneSpace
    alphaBoolBlend {blendOption=Linear}
    boneToModify {boneName=deform_lower_belt_front_right}
    rotation {yaw=10}
    translation=<-3 -1>
   }
   Save cached pose 'Post Belt Setup'(Pose=Component To Local(ComponentPose=@n68)) {cacheName=Post Belt Setup}
   n69 = Get R Arm Raise 45 to 95() -> (R Arm Raise 45 to 95:float)
   n70 = "Transform (Modify) Bone - Bone: dyn_collar_front_right_1"(
    ComponentPose=Local To Component(LocalPose=Use cached pose 'Post Belt Setup'())
    Alpha=@n69
    Translation="-3,0,-3"
    Rotation="-18.739,7.120,-21.167"
    Scale="1,1,1"
   ) {
    rotationMode=BMM_Additive
    rotationSpace=BCS_BoneSpace
    translationMode=BMM_Additive
    translationSpace=BCS_BoneSpace
    alphaBoolBlend {blendOption=Linear}
    alphaScaleBiasClamp {bInterpResult=true}
    boneToModify {boneName=dyn_collar_front_right_1}
    rotation=<rot -18.739 7.1199 -21.167>
    translation {x=-3, z=-3}
   }
   n71 = "Transform (Modify) Bone - Bone: dyn_collar_front_1"(ComponentPose=@n70, Alpha=@n69, Translation="-0.000,0.256,-1.806", Rotation="-20.012,4.976,0.025", Scale="1,1,1") {
    rotationMode=BMM_Additive
    rotationSpace=BCS_BoneSpace
    translationMode=BMM_Additive
    translationSpace=BCS_BoneSpace
    alphaBoolBlend {blendOption=Linear}
    alphaScaleBiasClamp {bInterpResult=true}
    boneToModify {boneName=dyn_collar_front_1}
    rotation=<rot -20.012 4.976 0.025>
    translation {y=0.256, z=-1.806}
   }
   n72 = "Transform (Modify) Bone - Bone: dyn_collar_back_right_1"(ComponentPose=@n71, Alpha=@n69, Translation="0,-3,-3", Rotation="0,0,0", Scale="1,1,1") {
    rotationMode=BMM_Additive
    rotationSpace=BCS_BoneSpace
    translationMode=BMM_Additive
    translationSpace=BCS_BoneSpace
    alphaBoolBlend {blendOption=Linear}
    alphaScaleBiasClamp {bInterpResult=true}
    boneToModify {boneName=dyn_collar_back_right_1}
    translation {y=-3, z=-3}
   }
   n73 = Get L Arm Raise 60 To 105() -> (L Arm Raise 60 To 105:float)
   n74 = "Transform (Modify) Bone - Bone: dyn_collar_back_left_1"(ComponentPose=@n72, Alpha=@n73, Translation="-1.168,0,-3", Rotation="-20.023,0.011,-0.003", Scale="1,1,1") {
    rotationMode=BMM_Additive
    rotationSpace=BCS_BoneSpace
    translationMode=BMM_Additive
    translationSpace=BCS_BoneSpace
    alphaBoolBlend {blendOption=Linear}
    alphaScaleBiasClamp {bInterpResult=true}
    boneToModify {boneName=dyn_collar_back_left_1}
    rotation=<rot -20.023 0.0111 -0.003>
    translation {x=-1.168, z=-3}
   }
   n75 = "Transform (Modify) Bone - Bone: dyn_collar_front_left_1"(ComponentPose=@n74, Alpha=@n73, Translation="-0.471,-0.236,1.453", Rotation="-14.977,0.036,-5.012", Scale="1,1,1") {
    rotationMode=BMM_Additive
    rotationSpace=BCS_BoneSpace
    translationMode=BMM_Additive
    translationSpace=BCS_BoneSpace
    alphaBoolBlend {blendOption=Linear}
    alphaScaleBiasClamp {bInterpResult=true}
    boneToModify {boneName=dyn_collar_front_left_1}
    rotation=<rot -14.9771 0.036 -5.012>
    translation=<-0.471 -0.236 1.453>
   }
   n76 = Get Spine rotation to the Left Alpha() -> (Spine rotation to the Left Alpha:double)
   n77 = "Transform (Modify) Bone - Bone: dyn_collar_front_1"(ComponentPose=@n75, Alpha=@n76, Translation="-0.067,2.988,0.000", Rotation="0,0,0", Scale="1,1,1") {
    rotationMode=BMM_Additive
    rotationSpace=BCS_BoneSpace
    translationMode=BMM_Additive
    translationSpace=BCS_BoneSpace
    alphaBoolBlend {blendOption=Linear}
    alphaScaleBiasClamp {bInterpResult=true}
    boneToModify {boneName=dyn_collar_front_1}
    translation=<-0.067 2.988>
   }
   n78 = "Transform (Modify) Bone - Bone: dyn_collar_front_right_1"(ComponentPose=@n77, Alpha=@n76, Translation="0.004,2.494,-3.022", Rotation="0,0,0", Scale="1,1,1") {
    rotationMode=BMM_Additive
    rotationSpace=BCS_BoneSpace
    translationMode=BMM_Additive
    translationSpace=BCS_BoneSpace
    alphaBoolBlend {blendOption=Linear}
    alphaScaleBiasClamp {bInterpResult=true}
    boneToModify {boneName=dyn_collar_front_right_1}
    translation=<0.004 2.494 -3.022>
   }
   n79 = Get Spine rotation to the Right Alpha() -> (Spine rotation to the Right Alpha:double)
   n80 = "Transform (Modify) Bone - Bone: dyn_collar_front_left_1"(ComponentPose=@n78, Alpha=@n79, Translation="0,-1.408,-1.628", Rotation="-10,0,-5", Scale="1,1,1") {
    rotationMode=BMM_Additive
    rotationSpace=BCS_BoneSpace
    translationMode=BMM_Additive
    translationSpace=BCS_BoneSpace
    alphaBoolBlend {blendOption=Linear}
    alphaScaleBiasClamp {bInterpResult=true}
    boneToModify {boneName=dyn_collar_front_left_1}
    rotation {pitch=-10, roll=-5}
    translation {y=-1.408, z=-1.628}
   }
   n81 = "Transform (Modify) Bone - Bone: dyn_collar_back_left_1"(ComponentPose=@n80, Alpha=@n79, Translation="0,0,-1", Rotation="0,0,0", Scale="1,1,1") {
    rotationMode=BMM_Additive
    rotationSpace=BCS_BoneSpace
    translationMode=BMM_Additive
    translationSpace=BCS_BoneSpace
    alphaBoolBlend {blendOption=Linear}
    alphaScaleBiasClamp {bInterpResult=true}
    boneToModify {boneName=dyn_collar_back_left_1}
    translation {z=-1}
   }
   n82 = "Transform (Modify) Bone - Bone: dyn_collar_back_right_1"(ComponentPose=@n81, Alpha=@n79, Translation="0,0,1", Rotation="0,0,0", Scale="1,1,1") {
    rotationMode=BMM_Additive
    rotationSpace=BCS_BoneSpace
    translationMode=BMM_Additive
    translationSpace=BCS_BoneSpace
    alphaBoolBlend {blendOption=Linear}
    alphaScaleBiasClamp {bInterpResult=true}
    boneToModify {boneName=dyn_collar_back_right_1}
    translation {z=1}
   }
   n83 = "Transform (Modify) Bone - Bone: dyn_collar_back_1"(ComponentPose=@n82, bAlphaBoolEnabled=True, Translation="-3,0,0", Rotation="0,-15,0", Scale="0.7,0.7,0.7") {
    alphaInputType=Bool
    rotationMode=BMM_Additive
    rotationSpace=BCS_BoneSpace
    scaleMode=BMM_Additive
    translationMode=BMM_Additive
    translationSpace=BCS_BoneSpace
    alphaBoolBlend {blendInTime=0.3, blendOption=ExpInOut, blendOutTime=0.3}
    alphaScaleBiasClamp {bInterpResult=true}
    boneToModify {boneName=dyn_collar_back_1}
    rotation {yaw=-15}
    scale=<0.7 0.7 0.7>
    translation {x=-3}
   }
   n84 = "Transform (Modify) Bone - Bone: dyn_collar_back_right_1"(ComponentPose=@n83, Alpha=1, Translation="-1.5,0,0", Rotation="0,0,0", Scale="0.6,0.6,0.6") {
    rotationMode=BMM_Additive
    rotationSpace=BCS_BoneSpace
    scaleMode=BMM_Additive
    translationMode=BMM_Additive
    translationSpace=BCS_BoneSpace
    alphaBoolBlend {blendInTime=0.3, blendOption=ExpInOut, blendOutTime=0.3}
    alphaScaleBiasClamp {interpSpeedDecreasing=20, interpSpeedIncreasing=20}
    boneToModify {boneName=dyn_collar_back_right_1}
    scale=<0.6 0.6 0.6>
    translation {x=-1.5}
   }
   n85 = "Transform (Modify) Bone - Bone: dyn_collar_back_right_1"(ComponentPose=@n84, Alpha=1, Translation="-3.5,0,0", Rotation="0,0,0", Scale="0.6,0.6,0.6") {
    rotationMode=BMM_Additive
    rotationSpace=BCS_BoneSpace
    scaleMode=BMM_Additive
    scaleSpace=BCS_BoneSpace
    translationMode=BMM_Additive
    translationSpace=BCS_BoneSpace
    alphaBoolBlend {blendOption=Linear}
    alphaScaleBiasClamp {interpSpeedDecreasing=20, interpSpeedIncreasing=20}
    boneToModify {boneName=dyn_collar_back_right_1}
    scale=<0.6 0.6 0.6>
    translation {x=-3.5}
   }
   n86 = "Transform (Modify) Bone - Bone: dyn_collar_back_1"!(ComponentPose=@n85, Alpha:float=1, Translation="-1.788,-0.038,-0.740", Rotation="-3.825,-29.194,-4.856", Scale="0.6,0.6,0.6") -> (Pose) {
    rotationMode=BMM_Additive
    rotationSpace=BCS_BoneSpace
    scaleMode=BMM_Additive
    scaleSpace=BCS_BoneSpace
    translationMode=BMM_Additive
    translationSpace=BCS_BoneSpace
    alphaBoolBlend {blendOption=Linear}
    alphaScaleBiasClamp {interpSpeedDecreasing=20, interpSpeedIncreasing=20}
    boneToModify {boneName=dyn_collar_back_1}
    rotation=<rot -3.8249 -29.194 -4.8561>
    scale=<0.6 0.6 0.6>
    translation=<-1.788 -0.038 -0.74>
   }
   n87 = "Transform (Modify) Bone - Bone: dyn_collar_back_left_1"(ComponentPose=@n86, Alpha=1, Translation="-3.827,-0.000,0.000", Rotation="0,0,0", Scale="1,1,1") {
    rotationMode=BMM_Additive
    rotationSpace=BCS_BoneSpace
    scaleMode=BMM_Additive
    translationMode=BMM_Additive
    translationSpace=BCS_BoneSpace
    alphaBoolBlend {blendOption=Linear}
    alphaScaleBiasClamp {interpSpeedDecreasing=20, interpSpeedIncreasing=20}
    boneToModify {boneName=dyn_collar_back_left_1}
    translation {x=-3.827}
   }
   n88 = "Transform (Modify) Bone - Bone: dyn_collar_back_1"(ComponentPose=@n87, Alpha=1, Translation="0,0,0", Rotation="0,0,0", Scale="1,1,1") {
    rotationMode=BMM_Additive
    rotationSpace=BCS_BoneSpace
    scaleMode=BMM_Additive
    translationMode=BMM_Additive
    translationSpace=BCS_BoneSpace
    alphaBoolBlend {blendOption=Linear}
    alphaScaleBiasClamp {interpSpeedDecreasing=20, interpSpeedIncreasing=20}
    boneToModify {boneName=dyn_collar_back_1}
   }
   n89 = "Transform (Modify) Bone - Bone: dyn_collar_back_1"(ComponentPose=@n88, Alpha=1, Translation="0,0,0", Rotation="0,0,-30", Scale="0.6,0.6,0.6") {
    rotationMode=BMM_Additive
    scaleMode=BMM_Additive
    translationMode=BMM_Additive
    alphaBoolBlend {blendOption=Linear}
    alphaScaleBiasClamp {interpSpeedDecreasing=20, interpSpeedIncreasing=20}
    boneToModify {boneName=dyn_collar_back_1}
    rotation {roll=-30}
    scale=<0.6 0.6 0.6>
   }
   n90 = "Transform (Modify) Bone - Bone: dyn_collar_front_right_1"(ComponentPose=@n89, Alpha=1, Translation="3,0,0", Rotation="0,0,-60", Scale="0.6,0.6,0.6") {
    rotationMode=BMM_Additive
    scaleMode=BMM_Additive
    translationMode=BMM_Additive
    alphaBoolBlend {blendOption=Linear}
    alphaScaleBiasClamp {interpSpeedDecreasing=20, interpSpeedIncreasing=20}
    boneToModify {boneName=dyn_collar_front_right_1}
    rotation {roll=-60}
    scale=<0.6 0.6 0.6>
    translation {x=3}
   }
   Save cached pose 'Post Collar Setup'(Pose=Component To Local(ComponentPose=@n90)) {cacheName=Post Collar Setup}
  ]}
 EventGraph {nodes [
   Event BlueprintUpdateAnimation()
   > n1 = Sequence!(execute) -> (then_0, then_1, then_2:Exec, then_3:Exec, then_4:Exec, then_5:Exec, then_6:Exec, then_7:Exec, then_8:Exec, then_9:Exec, then_10:Exec)
   n2 = Branch(execute=@n1.then_7, Condition=Get Is crouching and not moving() -> (Is crouching and not moving:bool))
   > Set Left Leg crouch Idle Bend Euler Z Alpha(Left Leg crouch Idle Bend Euler Z Alpha:double=Get Left Leg crouch Idle Bend Euler Z Alpha()) -> (Output_Get:double)
   > Set Melee Spine rotation to the Left Alpha(Melee Spine rotation to the Left Alpha:float=Get Melee Spine rotation to the Left Alpha()) -> (Output_Get:float)
   > Set Melee Spine rotation to the Right Alpha(Melee Spine rotation to the Right Alpha:float=Get Melee Spine rotation to the Right Alpha()) -> (Output_Get:float)
   Set Left Leg crouch Idle Bend Euler Z Alpha(execute=@n2.else, Left Leg crouch Idle Bend Euler Z Alpha=0)
   > Set Melee Spine rotation to the Left Alpha(Melee Spine rotation to the Left Alpha=0)
   > Set Melee Spine rotation to the Right Alpha(Melee Spine rotation to the Right Alpha=0)
   n3 = Branch(execute=@n1.then_8, Condition=Get Is crouching and moving() -> (Is crouching and moving:bool))
   > Set Left Leg crouch moving Bend Euler Z Alpha(Left Leg crouch moving Bend Euler Z Alpha:double=Get Left Leg crouch moving Bend Euler Z Alpha()) -> (Output_Get:double)
   Set Left Leg crouch moving Bend Euler Z Alpha(execute=@n3.else, Left Leg crouch moving Bend Euler Z Alpha=0)
   n4 = Branch(execute=@n1.then_9, Condition=NOT Boolean(A=Get bIsCrouching()))
   > Set Left Leg Bend Standing Euler Z Alpha(Left Leg Bend Standing Euler Z Alpha:double=Get Left Leg Bend Euler Z Alpha()) -> (Output_Get:double)
   > Set HeadLookUpWhileStandingAlpha(HeadLookUpWhileStandingAlpha:double=Get HeadLookUpAlpha() -> (HeadLookUpAlpha:double)) -> (Output_Get:double)
   Set Left Leg Bend Standing Euler Z Alpha(execute=@n4.else, Left Leg Bend Standing Euler Z Alpha=0)
   > Set HeadLookUpWhileStandingAlpha(HeadLookUpWhileStandingAlpha=0)
   n5 = Get Body Mesh() -> (Body Mesh:Skeletal Mesh Component Object Reference)
   n6 = Get bIsTargeting()
   Set R Arm Raise While Targeting(
    execute=@n1.then_0
    R Arm Raise While Targeting:float=float * float(
     A="To Float (Boolean)"(InBool=@n6)
     B=GetAlphaFromReferenceTransformDelta(SkeletalMeshComponent=@n5, BoneName=upperarm_r, BaseName=spine_05, InputRange=(X=45,Y=135), TwistAxis="0,0,1")
    ) {is=title K2Node_PromotableOperator}
   ) -> (Output_Get:float)
   > Set R Arm Raise 45 to 95(R Arm Raise 45 to 95:float=GetAlphaFromReferenceTransformDelta(SkeletalMeshComponent=@n5, BoneName=upperarm_r, BaseName=spine_05, InputRange=(X=45,Y=95), TwistAxis="0,0,1")) -> (Output_Get:float)
   > Set R Arm Raise 95 to 135(R Arm Raise 95 to 135:float=GetAlphaFromReferenceTransformDelta(SkeletalMeshComponent=@n5, BoneName=upperarm_r, BaseName=spine_05, InputRange=(X=95,Y=135), TwistAxis="0,0,1")) -> (Output_Get:float)
   > Set Melee Spine rotation to the Left Alpha(Melee Spine rotation to the Left Alpha=GetAlphaFromReferenceTransformDelta(SkeletalMeshComponent=@n5, BoneName=spine_02, BaseName=pelvis, DeltaComponentType=EulerX, InputRange=(X=-5,Y=7), TwistAxis="0,0,1"))
   > Set Melee Spine rotation to the Right Alpha(Melee Spine rotation to the Right Alpha=GetAlphaFromReferenceTransformDelta(SkeletalMeshComponent=@n5, BoneName=spine_01, BaseName=pelvis, DeltaComponentType=EulerX, InputRange=(X=-4,Y=-20), TwistAxis="0,0,1"))
   > Set LeftLegBendEulerY_Alpha(LeftLegBendEulerY_Alpha:double=GetAlphaFromReferenceTransformDelta(SkeletalMeshComponent=@n5, BoneName=thigh_l, BaseName=pelvis, DeltaComponentType=EulerY, InputRange=(X=10,Y=30), TwistAxis="0,0,1")) -> (Output_Get:double)
   > Set L Arm Raise 60 To 105(L Arm Raise 60 To 105:float=GetAlphaFromReferenceTransformDelta(SkeletalMeshComponent=@n5, BoneName=upperarm_l, BaseName=spine_05, InputRange=(X=60,Y=105), TwistAxis="0,0,1")) -> (Output_Get:float)
   > "Set L Arm Pivot Euler Y 0 To -16 "("L Arm Pivot Euler Y 0 To -16 ":float=GetAlphaFromReferenceTransformDelta(SkeletalMeshComponent=@n5, BoneName=upperarm_l, BaseName=spine_05, DeltaComponentType=EulerY, InputRange=(X=0,Y=-16), TwistAxis="0,0,1")) -> (Output_Get:float)
   > Set Spine rotation to the Left Alpha(Spine rotation to the Left Alpha:double=float * float(
     A=GetAlphaFromReferenceTransformDelta(SkeletalMeshComponent=@n5, BoneName=spine_05, BaseName=pelvis, DeltaComponentType=EulerX, InputRange=(X=-10,Y=7), TwistAxis="0,0,1")
     B:bool=Get bIsTargeting()
    )) -> (Output_Get:double)
   > Set Spine rotation to the Right Alpha(Spine rotation to the Right Alpha:double=float * float(
     A=GetAlphaFromReferenceTransformDelta(SkeletalMeshComponent=@n5, BoneName=spine_05, BaseName=pelvis, DeltaComponentType=EulerX, InputRange=(X=-10,Y=-45), TwistAxis="0,0,1")
     B=Get bIsTargeting()
    )) -> (Output_Get:double)
   > Set Not in Front End(Not in Front End:bool=NOT Boolean(A=Get bIsInFrontEnd())) -> (Output_Get:bool)
   n7 = Get Jumping or Falling Check() -> (Jumping or Falling Check:bool)
   n8 = Branch(execute=@n1.then_4, Condition=@n7)
   n9 = TryGetPawnOwner()
   n10 = GetVelocity!(self:Actor Object Reference=@n9) -> (ReturnValue:Vector)
   n11 = Dot Product!(A:Vector=@n10, B:Vector=GetActorRightVector!(self:Actor Object Reference=@n9) -> (ReturnValue:Vector)) -> (ReturnValue:double)
   n12 = Set Jumping to the Left Alpha(execute=@n31.then, Jumping to the Left Alpha:double=MapRangeClamped(Value=@n11, InRangeA=0, InRangeB=-230)) -> (Output_Get:double)
   Set Jumping to the Left Alpha(execute=@n8.then, Jumping to the Left Alpha=@n12)
   Set Jumping to the Left Alpha(execute=@n8.else, Jumping to the Left Alpha=0)
   n13 = Branch(execute=@n1.then_5, Condition=@n7)
   n14 = MapRangeClamped(Value=@n11, InRangeA=0, InRangeB=230)
   n15 = Set Jumping to the Right Alpha(execute=@n12.then, Jumping to the Right Alpha:double=@n14) -> (Output_Get:double)
   Set Jumping to the Right Alpha(execute=@n13.then, Jumping to the Right Alpha=@n15)
   Set Jumping to the Right Alpha(execute=@n13.else, Jumping to the Right Alpha=0)
   n16 = Get bIsCrouching()
   n17 = Branch(execute=@n1.then_10, Condition=@n16)
   > Set Crouching to the right Alpha(Crouching to the right Alpha:double=Get Crouching to the right Alpha() -> (Crouching to the right Alpha:double)) -> (Output_Get:double)
   Set Crouching to the right Alpha(execute=@n17.else, Crouching to the right Alpha=0)
   Set HeadLookUpAlpha(
    execute=@n1.then_1
    HeadLookUpAlpha:double=GetAlphaFromReferenceTransformDelta(SkeletalMeshComponent=@n5, BoneName=head, BaseName=spine_05, InputRange=(X=-25,Y=-50), TwistAxis="0,0,1")
   ) -> (Output_Get:double)
   > Set Head Tilt To Right Alpha(Head Tilt To Right Alpha:double=GetAlphaFromReferenceTransformDelta(
     SkeletalMeshComponent=@n5
     BoneName=head
     BaseName=spine_05
     DeltaComponentType=EulerY
     InputRange=(X=0,Y=30)
     OutputRange=(X=0,Y=0.7)
     TwistAxis="0,0,1"
    )) -> (Output_Get:double)
   > Set Left Leg Bend Euler Z Sprint Alpha(Left Leg Bend Euler Z Sprint Alpha:double=float * float(
     A:bool=Get bIsTacticalSprinting()
     B:double=GetAlphaFromReferenceTransformDelta(SkeletalMeshComponent=@n5, BoneName=thigh_l, BaseName=pelvis, InputRange=(X=0,Y=-105), TwistAxis="0,0,1")
    )) -> (Output_Get:double)
   n18 = GetAlphaFromReferenceTransformDelta(SkeletalMeshComponent=@n5, BoneName=thigh_l, BaseName=pelvis, InputRange=(X=0,Y=-105), TwistAxis="0,0,1")
   > Set Left Leg Bend Euler Z Alpha(Left Leg Bend Euler Z Alpha:double=@n18) -> (Output_Get:double)
   n19 = Get Is crouching and moving()
   > Set Left Leg Crouch Moving Forward or Backward Euler Z Alpha(Left Leg Crouch Moving Forward or Backward Euler Z Alpha:double=float * float!(
     A:double="Get Crouching forward of  backwards Alpha"() -> ("Crouching forward of  backwards Alpha":double)
     B="To Float (Boolean)"(InBool=@n19)
     C:float=@n18
    ) -> (ReturnValue)) -> (Output_Get:double)
   > Set Right Leg Bend Euler Z Alpha(Right Leg Bend Euler Z Alpha:double=GetAlphaFromReferenceTransformDelta(SkeletalMeshComponent=@n5, BoneName=thigh_r, BaseName=pelvis, InputRange=(X=0,Y=-60), TwistAxis="0,0,1")) -> (Output_Get:double)
   > Set Left Leg crouch Idle Bend Euler Z Alpha(Left Leg crouch Idle Bend Euler Z Alpha=float * float!(
     A=GetAlphaFromReferenceTransformDelta(SkeletalMeshComponent=@n5, BoneName=thigh_l, BaseName=pelvis, InputRange=(X=-50,Y=-120), TwistAxis="0,0,1")
     B:bool=Get Is crouching and not moving()
    ) -> (ReturnValue))
   > Set Right Leg crouch Bend Euler Z Alpha(Right Leg crouch Bend Euler Z Alpha:double=GetAlphaFromReferenceTransformDelta(SkeletalMeshComponent=@n5, BoneName=thigh_r, BaseName=pelvis, InputRange=(X=-50,Y=-130), TwistAxis="0,0,1")) -> (Output_Get:double)
   n20 = Get bIsCrouching()
   > Set Left Leg Bend to the Left Alpha(Left Leg Bend to the Left Alpha:double=float * float(
     A="To Float (Boolean)"(InBool=@n20)
     B:double=GetAlphaFromReferenceTransformDelta(SkeletalMeshComponent=@n5, BoneName=thigh_l, BaseName=pelvis, DeltaComponentType=EulerY, InputRange=(X=5,Y=40), TwistAxis="0,0,1")
    )) -> (Output_Get:double)
   > Set Left Leg Bend to the Right Alpha(Left Leg Bend to the Right Alpha:double=float * float(
     A="To Float (Boolean)"(InBool=@n20)
     B=GetAlphaFromReferenceTransformDelta(SkeletalMeshComponent=@n5, BoneName=thigh_l, BaseName=pelvis, DeltaComponentType=EulerY, InputRange=(X=6,Y=-13), TwistAxis="0,0,1")
    )) -> (Output_Get:double)
   n21 = float * float(
    A="To Float (Boolean)"(InBool=Get Is crouching and moving())
    B=GetAlphaFromReferenceTransformDelta(SkeletalMeshComponent=@n5, BoneName=thigh_l, BaseName=pelvis, InputRange=(X=-50,Y=-115), TwistAxis="0,0,1")
   )
   > Set Left Leg crouch moving Bend Euler Z Alpha(Left Leg crouch moving Bend Euler Z Alpha=@n21)
   > Set Left Leg crouch moving to the right Bend Euler Z Alpha(Left Leg crouch moving to the right Bend Euler Z Alpha:double=float * float(A=@n21, B=Get Crouching to the right Alpha())) -> (Output_Get:double)
   n22 = GetAlphaFromReferenceTransformDelta(SkeletalMeshComponent=@n5, BoneName=thigh_r, BaseName=pelvis, DeltaComponentType=EulerY, InputRange=(X=10,Y=35), TwistAxis="0,0,1")
   n23 = Get Right Leg Bend Euler Z Alpha()
   > Set Sidebag correction Alpha(Sidebag correction Alpha:double=float / float(A=float + float(A=@n23, B=@n22), B=2)) -> (Output_Get:double)
   n24 = GetAlphaFromReferenceTransformDelta(SkeletalMeshComponent=@n5, BoneName=thigh_r, BaseName=pelvis, InputRange=(X=0,Y=-60), TwistAxis="0,0,1")
   n25 = Get Is crouching and moving()
   > Set Right Leg crouch moving Bend Euler Z Alpha(Right Leg crouch moving Bend Euler Z Alpha:double=float * float(A="To Float (Boolean)"(InBool=@n25), B=@n24)) -> (Output_Get:double)
   n26 = Get Is crouching and not moving()
   > Set Right Leg crouch Idle Bend Euler Z Alpha(Right Leg crouch Idle Bend Euler Z Alpha:double=float * float(A="To Float (Boolean)"(InBool=@n26), B=@n24)) -> (Output_Get:double)
   n27 = "float > float"(A=Get Crouching to the right Alpha())
   > Set Left Leg Bend to the Right Alpha(Left Leg Bend to the Right Alpha=float * float(
     A="To Float (Boolean)"(InBool=@n27)
     B=GetAlphaFromReferenceTransformDelta(
      SkeletalMeshComponent=@n5
      BoneName=thigh_l
      BaseName=pelvis
      DeltaComponentType=EulerY
      InputRange=(X=6,Y=-13)
      OutputRange=(X=0,Y=0.6)
      TwistAxis="0,0,1"
     )
    ))
   n28 = NOT Boolean(A=Get bIsSkyDiving())
   Set Is crouching and moving(execute=@n1.then_2, Is crouching and moving:bool=AND Boolean!(A=Get bIsCrouching(), B=Get bIsMoving2D(), C:bool=@n28) -> (ReturnValue)) -> (Output_Get:bool)
   n29 = Get bIsMoving2D()
   > Set Is crouching and not moving(Is crouching and not moving:bool=AND Boolean(A=Get bIsCrouching(), B=NOT Boolean(A=@n29), C=@n28)) -> (Output_Get:bool)
   > n30 = Branch(Condition=Get bIsCrouching())
   > Set Crouching RArmRaise(Crouching RArmRaise:double=Get R Arm Raise 45 to 95()) -> (Output_Get:double)
   Set Crouching RArmRaise(execute=@n30.else, Crouching RArmRaise=0)
   n31 = Set Crouching to the left Alpha(execute=@n1.then_3, Crouching to the left Alpha:double=MapRangeClamped(Value=@n11, InRangeA=0, InRangeB=-230)) -> (Output_Get:double)
   n32 = Dot Product(A=@n10, B=GetActorForwardVector!(self:Actor Object Reference=@n9) -> (ReturnValue:Vector))
   Set Moving Backwards Alpha(execute=@n1.then_6, Moving Backwards Alpha:double=MapRangeClamped(Value=@n32, InRangeA=0, InRangeB=-300)) -> (Output_Get:double)
   > n33 = "Set Crouching forward of  backwards Alpha"("Crouching forward of  backwards Alpha":double=MapRangeClamped(Value="Absolute (Float)"(A=@n32), InRangeA=0, InRangeB=300)) -> (Output_Get:double)
   > n34 = Branch(Condition=Get bIsCrouching())
   > "Set Crouching forward of  backwards Alpha"("Crouching forward of  backwards Alpha"=@n33)
   "Set Crouching forward of  backwards Alpha"(execute=@n34.else, "Crouching forward of  backwards Alpha"=0)
   Event OnCharacterCustomizationLoaded()
   > Set Body Mesh(Body Mesh:Skeletal Mesh Component Object Reference=GetOwnerSkeletalMeshComponent()) -> (Output_Get:Skeletal Mesh Component Object Reference)
   float * float(A="To Float (Boolean)"(InBool=Get bIsTargeting()))
   Set Crouching to the left Alpha(execute=@n35.else, Crouching to the left Alpha=0)
   Set Crouching to the left Alpha(execute=@n35.then, Crouching to the left Alpha=@n31)
   n35 = Branch(execute=@n36.then, Condition=@n16)
   n36 = Set Crouching to the right Alpha(execute=@n15.then, Crouching to the right Alpha=@n14)
  ]}
]
variables [
 Body Mesh {type=SkeletalMeshComponent}
 "Crouching forward of  backwards Alpha" {type=number}
 Crouching RArmRaise {type=number}
 Crouching to the left Alpha {type=number}
 Crouching to the right Alpha {type=number}
 Head Tilt To Right Alpha {type=number}
 HeadLookUpAlpha {type=number}
 HeadLookUpWhileStandingAlpha {type=number}
 Is crouching and moving {type=boolean}
 Is crouching and not moving {type=boolean}
 Jumping or Falling Check {type=boolean}
 Jumping to the Left Alpha {type=number}
 Jumping to the Right Alpha {type=number}
 "L Arm Pivot Euler Y 0 To -16 " {type=number}
 L Arm Raise 60 To 105 {type=number}
 Left Leg Bend Euler Z Alpha {type=number}
 Left Leg Bend Euler Z Sprint Alpha {type=number}
 Left Leg Bend Standing Euler Z Alpha {type=number}
 Left Leg Bend to the Left Alpha {type=number}
 Left Leg Bend to the Right Alpha {type=number}
 Left Leg Bend to the Right Alpha_0 {type=number}
 Left Leg crouch Idle Bend Euler Z Alpha {type=number}
 Left Leg crouch moving Bend Euler Z Alpha {type=number}
 Left Leg Crouch Moving Forward or Backward Euler Z Alpha {type=number}
 Left Leg crouch moving to the right Bend Euler Z Alpha {type=number}
 LeftLegBendEulerY_Alpha {type=number}
 Melee Spine rotation to the Left Alpha {type=number}
 Melee Spine rotation to the Right Alpha {type=number}
 Moving Backwards Alpha {type=number}
 Not in Front End {type=boolean}
 R Arm Raise 45 to 95 {type=number}
 R Arm Raise 95 to 135 {type=number}
 R Arm Raise While Targeting {type=number}
 Right Leg Bend Euler Z Alpha {type=number}
 Right Leg crouch Bend Euler Z Alpha {type=number}
 Right Leg crouch Idle Bend Euler Z Alpha {type=number}
 Right Leg crouch moving Bend Euler Z Alpha {type=number}
 Sidebag correction Alpha {type=number}
 Spine rotation to the Left Alpha {type=number}
 Spine rotation to the Right Alpha {type=number}
]
