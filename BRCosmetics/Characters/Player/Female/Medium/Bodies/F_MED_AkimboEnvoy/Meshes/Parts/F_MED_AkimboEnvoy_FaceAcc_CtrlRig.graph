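# Control Rig for the AkimboEnvoy female face accessory that adjusts the ponytail bone chain, switching on the "is carrying" flag.
# When not carrying, four spherical pose readers track head orientation against spine_04 and weight corrective transforms on ponytail bones.
# When carrying, a four-point spline runs from the ponytail master through control bones on the neck and left clavicle, and the six-bone chain stretches to fit it.
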ControlRigBlueprint /BRCosmetics/Characters/Player/Female/Medium/Bodies/F_MED_AkimboEnvoy/Meshes/Parts/F_MED_AkimboEnvoy_FaceAcc_CtrlRig
graphs [
 RigVMFunctionLibrary {links [], nodes []}
 RigVMModel {nodes [
   RigUnit_BeginExecution = &RigUnit_BeginExecution()
   RigVMFunction_ControlFlowBranch = &RigVMFunction_ControlFlowBranch(ExecuteContext=@RigUnit_BeginExecution.ExecutePin, Condition=RigVMVariableNode() {variable=is carrying}.Value)
   n1 = &RigUnit_SphericalPoseReader(
    ActiveRegionScaleFactors=(PositiveWidth=1,NegativeWidth=2,PositiveHeight=2,NegativeHeight=2)
    ActiveRegionSize=0.1
    Debug=(bDrawDebug=false,bDraw2D=false,bDrawLocalAxes=false,DebugScale=25,DebugSegments=20,DebugThickness=0.25)
    DriverAxis=(X=0,Y=1,Z=0)
    DriverItem=(Type=Bone,Name="head")
    FalloffRegionScaleFactors=(PositiveWidth=0.75,NegativeWidth=1,PositiveHeight=2,NegativeHeight=2)
    FalloffSize=0.3
    FlipHeightScaling=False
    FlipWidthScaling=False
    OptionalParentItem=(Type=Bone,Name="spine_04")
    RotationOffset=(X=90,Y=0,Z=35)
    ExecutePin=@RigVMFunction_ControlFlowBranch.False
   )
   > &RigUnit_ModifyTransforms(
    ItemToModify="((Item=(Type=Bone,Name=\"dyn_ponytail_04\"),Transform=(Rotation=(X=0,Y=0,Z=0,W=1),Translation=(X=-3,Y=0,Z=0),Scale3D=(X=1,Y=1,Z=1))),(Item=(Type=Bone,Name=\"dyn_ponytail_03\"),Transform=(Rotation=(X=0,Y=0,Z=0,W=1),Translation=(X=-3,Y=0,Z=0),Scale3D=(X=1,Y=1,Z= ...(truncated)"
    WeightMinimum=0
    Weight=@n1.OutputParam
   )
   > n2 = &RigUnit_SphericalPoseReader(
    ActiveRegionScaleFactors=(PositiveWidth=2,NegativeWidth=1,PositiveHeight=2,NegativeHeight=2)
    ActiveRegionSize=0.1
    Debug=(bDrawDebug=false,bDraw2D=false,bDrawLocalAxes=false,DebugScale=25,DebugSegments=20,DebugThickness=0.25)
    DriverAxis=(X=0,Y=1,Z=0)
    DriverItem=(Type=Bone,Name="head")
    FalloffRegionScaleFactors=(PositiveWidth=1,NegativeWidth=0.75,PositiveHeight=1,NegativeHeight=1)
    FalloffSize=0.3
    FlipHeightScaling=False
    FlipWidthScaling=False
    OptionalParentItem=(Type=Bone,Name="spine_04")
    RotationOffset=(X=90,Y=0,Z=-60)
   )
   > &RigUnit_ModifyTransforms(
    ItemToModify="((Item=(Type=Bone,Name=\"dyn_ponytail_01\"),Transform=(Rotation=(X=0,Y=0,Z=0.2588,W=0.9659),Translation=(X=0,Y=0,Z=0),Scale3D=(X=1,Y=1,Z=1))),(Item=(Type=Bone,Name=\"dyn_ponytail_02\"),Transform=(Rotation=(X=0,Y=0,Z=-0.2588,W=0.9659),Translation=(X=0,Y=0,Z=0), ...(truncated)"
    WeightMinimum=0
    Weight=@n2.OutputParam
   )
   > n3 = &RigUnit_SphericalPoseReader(
    ActiveRegionScaleFactors=(PositiveWidth=1.5,NegativeWidth=1.5,PositiveHeight=0.7,NegativeHeight=0.7)
    ActiveRegionSize=0.1
    Debug=(bDrawDebug=false,bDraw2D=false,bDrawLocalAxes=false,DebugScale=25,DebugSegments=20,DebugThickness=0.25)
    DriverAxis=(X=0,Y=1,Z=0)
    DriverItem=(Type=Bone,Name="head")
    FalloffRegionScaleFactors=(PositiveWidth=1,NegativeWidth=1,PositiveHeight=1,NegativeHeight=1)
    FalloffSize=0.3
    FlipHeightScaling=False
    FlipWidthScaling=False
    OptionalParentItem=(Type=Bone,Name="spine_04")
    RotationOffset=(X=135,Y=0,Z=0)
   )
   > &RigUnit_ModifyTransforms(
    ItemToModify="((Item=(Type=Bone,Name=\"dyn_ponytail_03\"),Transform=(Rotation=(X=0,Y=0,Z=0,W=1),Translation=(X=0,Y=0,Z=3),Scale3D=(X=1,Y=1,Z=1))),(Item=(Type=Bone,Name=\"dyn_ponytail_04\"),Transform=(Rotation=(X=0,Y=0,Z=0,W=1),Translation=(X=0,Y=0,Z=3),Scale3D=(X=1,Y=1,Z=1) ...(truncated)"
    WeightMinimum=0
    Weight=@n3.OutputParam
   )
   > n4 = &RigUnit_SphericalPoseReader(
    ActiveRegionScaleFactors=(PositiveWidth=1.5,NegativeWidth=1.5,PositiveHeight=0.7,NegativeHeight=0.7)
    ActiveRegionSize=0.1
    Debug=(bDrawDebug=false,bDraw2D=false,bDrawLocalAxes=false,DebugScale=25,DebugSegments=20,DebugThickness=0.25)
    DriverAxis=(X=0,Y=1,Z=0)
    DriverItem=(Type=Bone,Name="head")
    FalloffRegionScaleFactors=(PositiveWidth=1,NegativeWidth=1,PositiveHeight=1,NegativeHeight=1)
    FalloffSize=0.3
    FlipHeightScaling=False
    FlipWidthScaling=False
    OptionalParentItem=(Type=Bone,Name="spine_04")
    RotationOffset=(X=45,Y=0,Z=0)
   )
   > &RigUnit_ModifyTransforms(
    ItemToModify="((Item=(Type=Bone,Name=\"dyn_ponytail_03\"),Transform=(Rotation=(X=0,Y=0,Z=0,W=1),Translation=(X=0,Y=0,Z=-5),Scale3D=(X=1,Y=1,Z=1))),(Item=(Type=Bone,Name=\"dyn_ponytail_04\"),Transform=(Rotation=(X=0,Y=0,Z=0,W=1),Translation=(X=0,Y=0,Z=-5),Scale3D=(X=1,Y=1,Z= ...(truncated)"
    WeightMinimum=0
    Weight=@n4.OutputParam
   )
   RigUnit_GetTransform = &RigUnit_GetTransform(Item=(Type=Bone,Name="dyn_con_ponytail_03"))
   SplineThrough4Points = RigVMFunctionReferenceNode(
    Point 3=@RigUnit_GetTransform.Transform.Translation
    Point 0=&RigUnit_GetTransform(Item=(Type=Bone,Name="dyn_ponytail_MSTR")).Transform.Translation
    Point 1=&RigUnit_GetTransform(Item=(Type=Bone,Name="dyn_con_ponytail_01")).Transform.Translation
    Point 2=&RigUnit_GetTransform(Item=(Type=Bone,Name="dyn_con_ponytail_02")).Transform.Translation
    Execute=@RigVMFunction_ControlFlowBranch.True
   )
   DISPATCH_RigVMDispatch_ArrayMake = RigVMDispatchNode(Values=((Type=Bone,Name="dyn_ponytail_01"),((),Type=Bone,Name="dyn_ponytail_02"),((),Type=Bone,Name="dyn_ponytail_03"),((),Type=Bone,Name="dyn_ponytail_04"),((),Type=Bone,Name="dyn_ponytail_05"),((),Type=Bone,Name="dyn_ponytail_06"))) {template="DISPATCH_RigVMDispatch_ArrayMake(in Values,out Array)"}
   /Script/ControlRigSpline.RigUnit_FitChainToSplineCurveItemArray(
    Alignment=Stretched
    DebugSettings=(bEnabled=False,Scale=1,CurveColor=(R=1,G=1,B=0,A=1),SegmentsColor=(R=1,G=0,B=0,A=1),WorldOffset=(Rotation=(X=0,Y=0,Z=0,W=1),Translation=(X=0,Y=0,Z=0),Scale3D=(X=1,Y=1,Z=1)))
    Maximum=1
    Minimum=0
    PoleVectorPosition=(X=0,Y=0,Z=0)
    PrimaryAxis=(X=1,Y=0,Z=0)
    RotationEaseType=Linear
    Rotations=()
    SamplingPrecision=12
    SecondaryAxis=(X=0,Y=0,Z=0)
    Weight=1
    bPropagateToChildren=True
    Items=@DISPATCH_RigVMDispatch_ArrayMake.Array
    ExecutePin=@SplineThrough4Points.Execute
    Spline=@SplineThrough4Points.Spline
   ) {is=struct RigVMUnitNode}
  ]}
]
hierarchy [
 root {loc=<0 0 0>, rot=<0 0 0 1>}
 pelvis {parent=root, loc=<0 -0.5301 98.0571>, rot=<0.0127 0.707 -0.0127 -0.707>}
 spine_01 {parent=pelvis, loc=<6.1208 -0.9328 0>, rot=<0 0 0.034 -0.9994>}
 spine_02 {parent=spine_01, loc=<6.5996 -0.2095 0>, rot=<0 0 0.0238 -0.9997>}
 spine_03 {parent=spine_02, loc=<6.4914 -0.1237 0>, rot=<0 0 0.0149 -0.9999>}
 spine_04 {parent=spine_03, loc=<6.9598 0.0794 0>, rot=<0 0 -0.0132 -0.9999>}
 spine_05 {parent=spine_04, loc=<6.9987 0 0>, rot=<0 0 -0.0015 -1>}
 neck_01 {parent=spine_05, loc=<11.3138 0.1529 0>, rot=<0 0 -0.2003 -0.9797>}
 neck_02 {parent=neck_01, loc=<4.5627 0 0>, rot=<0 0 0.0504 -0.9987>}
 head {parent=neck_02, loc=<5.3914 0 0>, rot=<0 0 0.115 -0.9934>}
 dyn_ponytail_MSTR {parent=head, loc=<10.139 -8.2373 -0.2605>, rot=<0 0 0.609 -0.7932>}
 dyn_ponytail_01 {parent=dyn_ponytail_MSTR, loc=<1.5085 0 0>, rot=<0.0087 0.0177 0.4404 -0.8976>}
 dyn_ponytail_02 {parent=dyn_ponytail_01, loc=<5.4259 0 0>, rot=<0.0157 -0.0016 0.4143 -0.91>}
 dyn_ponytail_03 {parent=dyn_ponytail_02, loc=<8.0318 0 0>, rot=<0.0011 0.003 0.0274 -0.9996>}
 dyn_ponytail_04 {parent=dyn_ponytail_03, loc=<8.6579 0 0>, rot=<-0.0009 -0.0033 -0.0228 -0.9997>}
 dyn_ponytail_05 {parent=dyn_ponytail_04, loc=<7.1193 0 0>, rot=<-0.0009 -0.0475 0.0712 -0.9963>}
 dyn_ponytail_06 {parent=dyn_ponytail_05, loc=<8.6642 0 0>, rot=<0 0 0 1>}
 dyn_stride_r_01 {parent=head, loc=<10.551 8.0716 3.0759>, rot=<0.131 -0.9244 -0.0502 0.3547>}
 dyn_stride_r_02_b {parent=dyn_stride_r_01, loc=<4.5543 -0.9363 -0.639>, rot=<-0.1082 0.1526 0.2266 -0.9559>}
 dyn_stride_r_02_a {parent=dyn_stride_r_01, loc=<4.333 1.6878 0.9718>, rot=<-0.3789 0.1833 0.024 -0.9068>}
 dyn_stride_l_01 {parent=head, loc=<10.7184 9.0125 -1.5905>, rot=<0.071 -0.8681 -0.13 -0.4738>}
 dyn_stride_l_02_b {parent=dyn_stride_l_01, loc=<6.3352 -2.2441 -0.7627>, rot=<0.1725 -0.4218 0.1445 -0.8783>}
 dyn_stride_l_02_a {parent=dyn_stride_l_01, loc=<5.5405 0.8025 -1.5897>, rot=<0.2099 -0.3338 0.1406 -0.9082>}
 dyn_con_ponytail_01 {parent=neck_02, loc=<5.7079 -11.2088 -7.5275>, rot=<-0.0784 -0.7027 0.0784 -0.7027>}
 dyn_con_ponytail_02 {parent=neck_01, loc=<3.8496 -1.7375 -11.4589>, rot=<-0.1137 -0.6979 0.1137 -0.6979>}
 clavicle_l {parent=spine_05, loc=<5.4831 -0.1053 -3.2946>, rot=<0.0465 -0.7289 -0.009 -0.6829>}
 upperarm_l {parent=clavicle_l, loc=<9.8658 -1.2136 -0.4105>, rot=<0.0164 -0.3245 -0.0269 -0.9454>}
 dyn_con_ponytail_03 {parent=clavicle_l, loc=<5.1122 7.8185 4.5955>, rot=<0.0008 0.0336 -0.0252 -0.9991>}
 clavicle_r {parent=spine_05, loc=<5.4831 -0.1053 3.2946>, rot=<0.7289 0.0465 -0.6829 0.009>}
 upperarm_r {parent=clavicle_r, loc=<-9.8658 1.2136 0.4105>, rot=<0.0164 -0.3245 -0.0269 -0.9454>}
]
variables [is carrying {type=bool, default=False}]
