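# Physics asset for the face accessory mesh of the female AkimboEnvoy cosmetic character, set up for a dynamic ponytail.
# Six capsule bodies form a constrained chain from a small sphere on the ponytail master bone, with swing limits widening toward the tip.
# Unconstrained head, pelvis, upper spine and upper arm shapes approximate the body, apparently as colliders for the hair.
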PhysicsAsset /BRCosmetics/Characters/Player/Female/Medium/Bodies/F_MED_AkimboEnvoy/Meshes/Parts/F_MED_AkimboEnvoy_FaceAcc_Physics
bodies {
 dyn_ponytail_01 [{shapeType=Capsule, length=2.5, radius=2, shapeName=dyn_ponytail_01_capsule, center=<2.9864 0 0>, rotation=<rot 90 0 0>}]
 dyn_ponytail_02 [{shapeType=Capsule, length=4.5, radius=2, shapeName=dyn_ponytail_02_capsule, center=<5.1432 0 0>, rotation=<rot 90 0 0>}]
 dyn_ponytail_03 [{shapeType=Capsule, length=3, radius=2, shapeName=dyn_ponytail_03_capsule, center=<4.4283 0 0>, rotation=<rot 90 90 90.0001>}]
 dyn_ponytail_04 [{shapeType=Capsule, length=3, radius=2, shapeName=dyn_ponytail_04_capsule, center=<3 0 0>, rotation=<rot 90 0 0>}]
 dyn_ponytail_05 [{shapeType=Capsule, length=3, radius=2, shapeName=dyn_ponytail_05_capsule, center=<4.0165 0 0>, rotation=<rot 90 0 0>}]
 dyn_ponytail_06 [{shapeType=Capsule, length=6, radius=2, shapeName=dyn_ponytail_06_capsule, center=<6.7788 0 0>, rotation=<rot 90 0 0>}]
 dyn_ponytail_MSTR [{shapeType=Sphere, radius=1, shapeName=dyn_ponytail_MSTR_sphere}]
 head [{shapeType=Sphere, radius=11, shapeName=head_sphere}]
 pelvis [{shapeType=Capsule, length=16, radius=14, shapeName=pelvis_capsule}]
 spine_04 [{shapeType=Capsule, length=29.5, radius=18, shapeName=spine_04_capsule, center=<-1.2785 7.8812 0>, rotation=<rot -90 -180 -180>}]
 spine_05 [{shapeType=Capsule, length=12, radius=14, shapeName=spine_05_capsule, center=<11.9751 5.2979 0>}]
 upperarm_l [{shapeType=Capsule, length=23.5, radius=6.5, shapeName=upperarm_l_capsule, center=<14.1088 0 0>, rotation=<rot -90 179.9998 -180>}]
 upperarm_r [{shapeType=Capsule, length=23.5, radius=6.5, shapeName=upperarm_r_capsule, center=<-18.3193 0 0>, rotation=<rot -90 90 -90>}]
}
constraints [
 {bone1Name=dyn_ponytail_01, bone2Name=dyn_ponytail_MSTR, swing1LimitDegrees=15, swing2LimitDegrees=15, twistLimitDegrees=10}
 {bone1Name=dyn_ponytail_02, bone2Name=dyn_ponytail_01, swing1LimitDegrees=20, swing2LimitDegrees=15, twistLimitDegrees=10}
 {bone1Name=dyn_ponytail_03, bone2Name=dyn_ponytail_02, swing1LimitDegrees=20, swing2LimitDegrees=20, twistLimitDegrees=5}
 {bone1Name=dyn_ponytail_04, bone2Name=dyn_ponytail_03, swing1LimitDegrees=30, swing2LimitDegrees=25, twistLimitDegrees=15}
 {bone1Name=dyn_ponytail_05, bone2Name=dyn_ponytail_04, swing1LimitDegrees=40, swing2LimitDegrees=25, twistLimitDegrees=15}
 {bone1Name=dyn_ponytail_06, bone2Name=dyn_ponytail_05, swing1LimitDegrees=60, swing2LimitDegrees=60, twistLimitDegrees=5}
]
