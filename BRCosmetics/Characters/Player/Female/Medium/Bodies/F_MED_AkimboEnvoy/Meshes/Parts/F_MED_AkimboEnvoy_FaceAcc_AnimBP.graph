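# AnimBlueprint for the AkimboEnvoy FaceAcc part, adding secondary motion to front hair strands and the ponytail over the pose copied from the attached parent.
# Front strands use AnimDynamics; the ponytail uses RigidBody, or AnimDynamics when rigid bodies are disabled.
# Backpacks or capes switch to constrained poses with Control Rig, while DBNO carrying uses Control Rig alone.
# Event graph updates gravity, backbling flags, head-tilt alpha and a low-FPS alpha that steadies the RigidBody ponytail.
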AnimBlueprint /BRCosmetics/Characters/Player/Female/Medium/Bodies/F_MED_AkimboEnvoy/Meshes/Parts/F_MED_AkimboEnvoy_FaceAcc_AnimBP
target_skeleton=&/F_MED_AkimboEnvoy_FaceAcc_Skeleton
graphs [
 AnimGraph {nodes [
   n1 = Get LowFPSAlpha() -> (LowFPSAlpha:float)
   n2 = Look At - Bone: dyn_ponytail_01(ComponentPose=Local To Component(LocalPose=Use cached pose 'PostRBAN'()), Alpha=@n1) {alphaBoolBlend {blendOption=Linear}, boneToModify {boneName=dyn_ponytail_01}, lookAtTarget {boneReference {boneName=pelvis}}}
   n3 = Look At - Bone: dyn_ponytail_02(ComponentPose=@n2, Alpha=@n1) {alphaBoolBlend {blendOption=Linear}, boneToModify {boneName=dyn_ponytail_02}, lookAtTarget {boneReference {boneName=pelvis}}}
   n4 = "Transform (Modify) Bone - Bone: dyn_ponytail_MSTR"!(ComponentPose=@n3, bAlphaBoolEnabled:bool=True, Translation="0,0,0", Rotation="0,30,0", Scale="1,1,1") -> (Pose) {
    alphaInputType=Bool
    rotationMode=BMM_Additive
    rotationSpace=BCS_BoneSpace
    alphaBoolBlend {blendOption=Linear, blendOutTime=0.5}
    boneToModify {boneName=dyn_ponytail_MSTR}
    rotation {yaw=30}
   }
   n5 = Blend Poses by bool(
    BlendPose_0=Component To Local(ComponentPose=@n4)
    BlendPose_1=Use cached pose 'BackHairAnimDynamics'()
    bActiveValue=Get bRigidBodyEnabled()
   ) {blendType=Linear}
   n6 = Use cached pose 'BackpackDynamics'()
   n7 = Get bIsWearingBackpack() -> (bIsWearingBackpack:bool)
   Output Pose(Result=Blend Poses by bool(
     BlendPose_0=Use cached pose 'dbno pose'()
     BlendPose_1=Blend Poses by bool(BlendPose_0=@n6, BlendPose_1=@n5, bActiveValue=@n7) {blendType=Linear}
     bActiveValue=Get bIsDBNOCarrying() -> (bIsDBNOCarrying:bool)
    ) {blendType=Linear}) {layerGroup=None}
   Save cached pose 'InputPose'(Pose=Copy Pose From Mesh() {bUseAttachedParent=true}) {cacheName=InputPose}
   n8 = Look At - Bone: dyn_ponytail_02(ComponentPose=Local To Component(LocalPose=Use cached pose 'PostFrontHair'()), Alpha=1) {
    alphaBoolBlend {blendOption=Linear}
    boneToModify {boneName=dyn_ponytail_02}
    lookAtLocation {x=0}
    lookAtTarget {boneReference {boneName=dyn_ponytail_03}}
    lookAt_Axis {axis=<1 0>}
   }
   n9 = "Constraint - dyn_ponytail_03 (PPPPP)"!(
    ComponentPose=@n8
    Alpha=1
    ConstraintWeights_0=0.4559
    ConstraintWeights_1:float=0.0024
    ConstraintWeights_2:float=0.1264
    ConstraintWeights_3:float=0.1338
    ConstraintWeights_4:float=0.2815
   ) -> (Pose) {
    alphaBoolBlend {blendOption=Linear}
    boneToModify {boneName=dyn_ponytail_03}
    constraintSetup [
     {targetBone {boneName=spine_05}}
     {targetBone {boneName=clavicle_l}}
     {targetBone {boneName=neck_01}}
     {targetBone {boneName=head}}
     {targetBone {boneName=neck_02}}
    ]
    constraintWeights [1,1,1,1,1]
   }
   n10 = "Constraint - dyn_ponytail_04 (PPP)"!(ComponentPose=@n9, Alpha=1, ConstraintWeights_0=0.6801, ConstraintWeights_1:float=0.1844, ConstraintWeights_2:float=0.1356) -> (Pose) {
    alphaBoolBlend {blendOption=Linear}
    boneToModify {boneName=dyn_ponytail_04}
    constraintSetup [{targetBone {boneName=spine_05}}, {targetBone {boneName=clavicle_l}}, {targetBone {boneName=clavicle_r}}]
    constraintWeights [1,1,1]
   }
   Save cached pose 'PostUpBraidConstrained'(Pose=Component To Local(ComponentPose=@n10)) {cacheName=PostUpBraidConstrained}
   n11 = Anim Dynamics - Bone: dyn_stride_l_01!(ComponentPose=Local To Component(LocalPose=Use cached pose 'InputPose'()), Alpha=1, GravityOverride:Vector="0,0,-980", ExternalForce="0,0,0") -> (Pose) {
    angularDampingOverride=1
    angularSpringConstant=1500
    bOverrideAngularDamping=true
    bUseGravityOverride=true
    gravityScale=0.35
    simulationSpace=RootRelative
    alphaBoolBlend {blendOption=Linear}
    boundBone {boneName=dyn_stride_l_01}
    componentAppliedLinearAccClamp=<250 250 250>
    componentLinearAccScale=<0.2 0.2 0.2>
    componentLinearVelScale=<0.2 0.2 0.2>
    gravityOverride {z=-980}
    physicsBodyDefinitions [{
      sphereCollisionRadius=0
      boundBone {boneName=dyn_stride_l_01}
      boxExtents=<15 15 15>
      constraintSetup {angularTargetAxis=AxisY, angularLimitsMax=<0 0 5>, angularLimitsMin=<0 -10 -5>, angularTarget=<0 -10 0>}
      localJointOffset=<2 0 0>
     }]
   }
   n12 = Anim Dynamics - Bone: dyn_stride_l_02_b!(ComponentPose=@n11, Alpha=1, GravityOverride:Vector="0,0,-980", ExternalForce="0,0,0") -> (Pose) {
    angularDampingOverride=0.9999
    angularSpringConstant=10
    bAngularSpring=true
    bOverrideAngularDamping=true
    bUseGravityOverride=true
    gravityScale=0.35
    simulationSpace=RootRelative
    alphaBoolBlend {blendOption=Linear}
    boundBone {boneName=dyn_stride_l_02_b}
    componentAppliedLinearAccClamp=<600 600 600>
    componentLinearAccScale=<0.25 0.25 0.25>
    componentLinearVelScale=<0.2 0.2 0.2>
    gravityOverride {z=-980}
    physicsBodyDefinitions [{
      sphereCollisionRadius=0
      boundBone {boneName=dyn_stride_l_02_b}
      boxExtents=<7 8 10>
      constraintSetup {angularTargetAxis=AxisY, angularLimitsMax=<5 0 25>, angularLimitsMin=<-5 -20 -5>, angularTarget=<0 -10 0>}
      localJointOffset=<3 0 0>
     }]
   }
   n13 = Anim Dynamics - Bone: dyn_stride_l_02_a!(ComponentPose=@n12, Alpha=1, GravityOverride:Vector="0,0,-980", ExternalForce="0,0,0") -> (Pose) {
    angularDampingOverride=0.999
    angularSpringConstant=10
    bAngularSpring=true
    bOverrideAngularDamping=true
    bUseGravityOverride=true
    gravityScale=0.35
    simulationSpace=RootRelative
    alphaBoolBlend {blendOption=Linear}
    boundBone {boneName=dyn_stride_l_02_a}
    componentAppliedLinearAccClamp=<350 350 350>
    componentLinearAccScale=<0.3 0.3 0.3>
    componentLinearVelScale=<0.2 0.2 0.2>
    gravityOverride {z=-980}
    physicsBodyDefinitions [{
      sphereCollisionRadius=0
      boundBone {boneName=dyn_stride_l_02_a}
      boxExtents=<8 10 10>
      constraintSetup {angularTargetAxis=AxisY, angularLimitsMax=<5 0 20>, angularLimitsMin=<-5 -30 -25>, angularTarget=<0 -10 0>}
      localJointOffset=<2.5 0 0>
     }]
   }
   n14 = Anim Dynamics - Bone: dyn_stride_r_01!(ComponentPose=@n13, Alpha=1, GravityOverride:Vector="0,0,-980", ExternalForce="0,0,0") -> (Pose) {
    angularDampingOverride=1
    angularSpringConstant=1500
    bOverrideAngularDamping=true
    bUseGravityOverride=true
    gravityScale=0.35
    simulationSpace=RootRelative
    alphaBoolBlend {blendOption=Linear}
    boundBone {boneName=dyn_stride_r_01}
    componentAppliedLinearAccClamp=<150 150 150>
    componentLinearAccScale=<0.2 0.2 0.2>
    componentLinearVelScale=<0.3 0.3 0.3>
    gravityOverride {z=-980}
    physicsBodyDefinitions [{
      sphereCollisionRadius=0
      boundBone {boneName=dyn_stride_r_01}
      boxExtents=<12 15 15>
      constraintSetup {angularTargetAxis=AxisY, angularLimitsMax=<0 10 5>, angularLimitsMin=<0 0 -5>, angularTarget=<0 -10 0>}
      localJointOffset=<1 0 0>
     }]
    relativeSpaceBone {boneName=root}
   }
   n15 = Anim Dynamics - Bone: dyn_stride_r_02_b!(ComponentPose=@n14, Alpha=1, GravityOverride:Vector="0,0,-980", ExternalForce="0,0,0") -> (Pose) {
    angularDampingOverride=1
    angularSpringConstant=10
    bAngularSpring=true
    bOverrideAngularDamping=true
    bUseGravityOverride=true
    gravityScale=0.35
    simulationSpace=RootRelative
    alphaBoolBlend {blendOption=Linear}
    boundBone {boneName=dyn_stride_r_02_b}
    componentAppliedLinearAccClamp=<300 300 300>
    componentLinearAccScale=<0.2 0.2 0.2>
    componentLinearVelScale=<0.25 0.25 0.25>
    gravityOverride {z=-980}
    physicsBodyDefinitions [{
      sphereCollisionRadius=0
      boundBone {boneName=dyn_stride_r_02_b}
      boxExtents=<10 10 8>
      constraintSetup {angularTargetAxis=AxisY, angularLimitsMax=<5 15 15>, angularLimitsMin=<-5 0 -10>, angularTarget=<0 -10 0>}
      localJointOffset=<2 0 0>
     }]
    relativeSpaceBone {boneName=root}
   }
   n16 = Anim Dynamics - Bone: dyn_stride_r_02_a!(ComponentPose=@n15, Alpha=1, GravityOverride:Vector="0,0,-980", ExternalForce="0,0,0") -> (Pose) {
    angularDampingOverride=1
    angularSpringConstant=10
    bAngularSpring=true
    bOverrideAngularDamping=true
    bUseGravityOverride=true
    gravityScale=0.35
    simulationSpace=RootRelative
    alphaBoolBlend {blendOption=Linear}
    boundBone {boneName=dyn_stride_r_02_a}
    componentAppliedLinearAccClamp=<600 600 600>
    componentLinearAccScale=<0.2 0.2 0.2>
    componentLinearVelScale=<0.2 0.2 0.2>
    gravityOverride {z=-980}
    physicsBodyDefinitions [{
      sphereCollisionRadius=0
      boundBone {boneName=dyn_stride_r_02_a}
      boxExtents=<7 8 10>
      constraintSetup {angularTargetAxis=AxisY, angularLimitsMax=<5 10 15>, angularLimitsMin=<-5 0 -25>, angularTarget=<0 -10 0>}
      localJointOffset=<2 0 0>
     }]
    relativeSpaceBone {boneName=root}
   }
   Save cached pose 'PostFrontHair'(Pose=Component To Local(ComponentPose=@n16)) {cacheName=PostFrontHair}
   n17 = "Transform (Modify) Bone - Bone: dyn_ponytail_MSTR"!(
    ComponentPose=Local To Component(LocalPose=Use cached pose 'PostFrontHair'())
    Alpha:float=Get HeadUpAlpha() -> (HeadUpAlpha:double)
    Translation="0,0,0"
    Rotation="0,30,0"
    Scale="1,1,1"
   ) -> (Pose) {
    rotationMode=BMM_Additive
    rotationSpace=BCS_BoneSpace
    alphaBoolBlend {blendOption=Linear, blendOutTime=0.5}
    boneToModify {boneName=dyn_ponytail_MSTR}
    rotation {yaw=30}
   }
   n18 = RigidBody!(ComponentPose=@n17, Alpha, OverrideWorldGravity:Vector=Get Gravity Override(), ExternalForce) -> (Pose) {
    bOverrideWorldGravity=true
    overridePhysicsAsset=&/F_MED_AkimboEnvoy_FaceAcc_Physics
    simulationSpace=BaseBoneSpace
    alphaBoolBlend {blendOption=Linear}
    baseBoneRef {boneName=neck_02}
    componentAppliedLinearAccClamp=<1000 650 400>
    componentLinearAccScale=<0.4 0.6 0.4>
    componentLinearVelScale=<0.8 0.4 0.2>
    overrideWorldGravity {z=-980}
   }
   Save cached pose 'PostRBAN'(Pose=Component To Local(ComponentPose=@n18)) {cacheName=PostRBAN}
   n19 = "Constraint - dyn_ponytail_05 (P)"!(
    ComponentPose=Local To Component(LocalPose=Use cached pose 'PostUpBraidConstrained'())
    Alpha=1
    ConstraintWeights_0=1
    ConstraintWeights_1:float=0
   ) -> (Pose) {
    alphaBoolBlend {blendOption=Linear}
    boneToModify {boneName=dyn_ponytail_05}
    constraintSetup [{targetBone {boneName=spine_05}}]
    constraintWeights [1,0]
   }
   n20 = "Constraint - dyn_ponytail_06 (PPPPPPP)"!(
    ComponentPose=@n19
    Alpha=1
    ConstraintWeights_0=0.0267
    ConstraintWeights_1:float=0.3785
    ConstraintWeights_2:float=0.0489
    ConstraintWeights_3:float=0.2925
    ConstraintWeights_4:float=0.127
    ConstraintWeights_5:float=0.1247
    ConstraintWeights_6:float=0.0015
   ) -> (Pose) {
    alphaBoolBlend {blendOption=Linear}
    boneToModify {boneName=dyn_ponytail_06}
    constraintSetup [
     {targetBone {boneName=spine_02}}
     {targetBone {boneName=spine_04}}
     {targetBone {boneName=spine_03}}
     {targetBone {boneName=spine_05}}
     {targetBone {boneName=clavicle_l}}
     {targetBone {boneName=clavicle_r}}
     {targetBone {boneName=spine_01}}
    ]
    constraintWeights [1,1,1,1,1,1,1]
   }
   n21 = "Transform (Modify) Bone - Bone: dyn_ponytail_01"!(ComponentPose=@n20, Alpha=1, Rotation="0,-8,0") -> (Pose) {
    rotationMode=BMM_Additive
    rotationSpace=BCS_BoneSpace
    scaleMode=BMM_Replace
    scaleSpace=BCS_BoneSpace
    alphaBoolBlend {blendOption=Linear}
    boneToModify {boneName=dyn_ponytail_01}
    rotation {yaw=-8}
   }
   n22 = "Transform (Modify) Bone - Bone: dyn_ponytail_03"!(ComponentPose=@n21, Alpha=1, Rotation="0,13,0") -> (Pose) {
    rotationMode=BMM_Additive
    rotationSpace=BCS_BoneSpace
    scaleMode=BMM_Replace
    scaleSpace=BCS_BoneSpace
    alphaBoolBlend {blendOption=Linear}
    boneToModify {boneName=dyn_ponytail_03}
    rotation {yaw=13}
   }
   n23 = "Transform (Modify) Bone - Bone: dyn_ponytail_04"!(ComponentPose=@n22, Alpha=1, Rotation="0,0,0", Scale="1,0.8,0.8") -> (Pose) {
    rotationMode=BMM_Additive
    rotationSpace=BCS_BoneSpace
    scaleMode=BMM_Replace
    scaleSpace=BCS_BoneSpace
    alphaBoolBlend {blendOption=Linear}
    boneToModify {boneName=dyn_ponytail_04}
    scale {y=0.8, z=0.8}
   }
   n24 = "Transform (Modify) Bone - Bone: dyn_ponytail_04"!(
    ComponentPose=Local To Component(LocalPose=Use cached pose 'PostUpBraidConstrained'())
    Alpha=1
    Translation:Vector="0,0,0"
    Rotation="0,5,0"
    Scale="1,1,1"
   ) -> (Pose) {
    rotationMode=BMM_Additive
    rotationSpace=BCS_BoneSpace
    scaleMode=BMM_Replace
    scaleSpace=BCS_BoneSpace
    translationSpace=BCS_BoneSpace
    alphaBoolBlend {blendInTime=0.3, blendOption=Linear, blendOutTime=0.3}
    boneToModify {boneName=dyn_ponytail_04}
    rotation {yaw=5}
   }
   n25 = "Constraint - dyn_ponytail_05 (P)"(ComponentPose=@n24, Alpha=1, ConstraintWeights_0=1, ConstraintWeights_1=0) {
    alphaBoolBlend {blendOption=Linear}
    boneToModify {boneName=dyn_ponytail_05}
    constraintSetup [{targetBone {boneName=spine_05}}]
    constraintWeights [1,0]
   }
   n26 = Anim Dynamics - Bone: dyn_ponytail_06!(ComponentPose=@n25, Alpha=1, GravityOverride:Vector="0,0,-980", ExternalForce="0,0,0") -> (Pose) {
    angularDampingOverride=1
    angularSpringConstant=1500
    bOverrideAngularDamping=true
    bUseGravityOverride=true
    gravityScale=0.35
    simulationSpace=RootRelative
    alphaBoolBlend {blendOption=Linear}
    boundBone {boneName=dyn_ponytail_06}
    componentAppliedLinearAccClamp=<500 500 500>
    componentLinearAccScale=<0.2 0.2 0.2>
    componentLinearVelScale=<0.2 0.2 0.2>
    gravityOverride {z=-980}
    physicsBodyDefinitions [{
      sphereCollisionRadius=0
      boundBone {boneName=dyn_ponytail_06}
      boxExtents=<10 15 15>
      constraintSetup {angularTargetAxis=AxisY, angularLimitsMax=<0 10 10>, angularLimitsMin=<0 -10 0>, angularTarget=<0 -10 0>}
      localJointOffset=<3 0 0>
     }]
   }
   n27 = Control Rig!(
    Source=Blend Poses by bool(
     BlendPose_0=Component To Local(ComponentPose=@n23)
     BlendPose_1=Component To Local(ComponentPose=@n26)
     bActiveValue=Get bIsWearingCape() -> (bIsWearingCape:bool)
    ) {blendType=Linear}
    Alpha
    is carrying:bool=Get bIsDBNOCarrying()
   ) -> (Pose) {
    alphaBoolBlend {blendOption=Linear}
    controlRigAssetReference {blueprintRigClass=&/F_MED_AkimboEnvoy_FaceAcc_CtrlRig.F_MED_AkimboEnvoy_FaceAcc_CtrlRig_C}
   }
   Save cached pose 'BackpackDynamics'(Pose=@n27) {cacheName=BackpackDynamics}
   n28 = Look At - Bone: dyn_ponytail_MSTR(ComponentPose=Local To Component(LocalPose=Use cached pose 'PostFrontHair'()), Alpha=1) {
    alphaBoolBlend {blendOption=Linear}
    boneToModify {boneName=dyn_ponytail_MSTR}
    lookAtLocation {x=0}
    lookAtTarget {boneReference {boneName=pelvis}}
    lookAt_Axis {axis {y=-1}}
   }
   n29 = "Transform (Modify) Bone - Bone: dyn_ponytail_01"!(ComponentPose=@n28, Alpha=1, Translation:Vector="0,0,0", Rotation="0,35,0", Scale:Vector="1,1,1") -> (Pose) {
    rotationMode=BMM_Additive
    rotationSpace=BCS_BoneSpace
    alphaBoolBlend {blendOption=Linear}
    boneToModify {boneName=dyn_ponytail_01}
    rotation {yaw=35}
   }
   n30 = Anim Dynamics - Bone: dyn_ponytail_01!(ComponentPose=@n29, Alpha=1, GravityOverride:Vector="0,0,-980", ExternalForce="0,0,0") -> (Pose) {
    angularDampingOverride=0.85
    angularSpringConstant=1500
    bOverrideAngularDamping=true
    bUseGravityOverride=true
    gravityScale=0.35
    simulationSpace=RootRelative
    alphaBoolBlend {blendOption=Linear}
    boundBone {boneName=dyn_ponytail_01}
    componentAppliedLinearAccClamp=<250 250 1100>
    componentLinearAccScale=<0.2 0.2 0.8>
    componentLinearVelScale=<0.4 0.4 1>
    gravityOverride {z=-980}
    physicsBodyDefinitions [{
      sphereCollisionRadius=0
      boundBone {boneName=dyn_ponytail_01}
      boxExtents=<15 15 15>
      constraintSetup {angularTargetAxis=AxisY, angularLimitsMax=<3 10 25>, angularLimitsMin=<-3 -10 0>, angularTarget=<0 -10 0>}
      localJointOffset=<1 0 0>
     }]
   }
   n31 = Anim Dynamics - Bone: dyn_ponytail_02!(ComponentPose=@n30, Alpha=1, GravityOverride:Vector="0,0,-980", ExternalForce="0,0,0") -> (Pose) {
    angularDampingOverride=0.85
    angularSpringConstant=1500
    bOverrideAngularDamping=true
    bUseGravityOverride=true
    gravityScale=0.35
    simulationSpace=RootRelative
    alphaBoolBlend {blendOption=Linear}
    boundBone {boneName=dyn_ponytail_02}
    componentAppliedLinearAccClamp=<250 250 1100>
    componentLinearAccScale=<0.2 0.2 0.8>
    componentLinearVelScale=<0.4 0.4 1>
    gravityOverride {z=-980}
    physicsBodyDefinitions [{
      sphereCollisionRadius=0
      boundBone {boneName=dyn_ponytail_02}
      boxExtents=<15 15 15>
      constraintSetup {angularTargetAxis=AxisY, angularLimitsMax=<3 10 25>, angularLimitsMin=<-3 -10 0>, angularTarget=<0 -10 0>}
      localJointOffset=<1.5 0 0>
     }]
   }
   n32 = Anim Dynamics - Bone: dyn_ponytail_03!(ComponentPose=@n31, Alpha=1, GravityOverride:Vector="0,0,-980", ExternalForce="0,0,0") -> (Pose) {
    angularDampingOverride=0.85
    angularSpringConstant=1500
    bOverrideAngularDamping=true
    bUseGravityOverride=true
    gravityScale=0.35
    simulationSpace=RootRelative
    alphaBoolBlend {blendOption=Linear}
    boundBone {boneName=dyn_ponytail_03}
    componentAppliedLinearAccClamp=<250 250 1100>
    componentLinearAccScale=<0.2 0.2 0.8>
    componentLinearVelScale=<0.4 0.4 1>
    gravityOverride {z=-980}
    physicsBodyDefinitions [{
      sphereCollisionRadius=0
      boundBone {boneName=dyn_ponytail_03}
      boxExtents=<15 15 15>
      constraintSetup {angularTargetAxis=AxisY, angularLimitsMax=<3 10 25>, angularLimitsMin=<-3 -10 0>, angularTarget=<0 -10 0>}
      localJointOffset=<2 0 0>
     }]
   }
   n33 = Anim Dynamics - Bone: dyn_ponytail_04!(ComponentPose=@n32, Alpha=1, GravityOverride:Vector="0,0,-980", ExternalForce="0,0,0") -> (Pose) {
    angularDampingOverride=0.85
    angularSpringConstant=1500
    bOverrideAngularDamping=true
    bUseGravityOverride=true
    gravityScale=0.35
    simulationSpace=RootRelative
    alphaBoolBlend {blendOption=Linear}
    boundBone {boneName=dyn_ponytail_04}
    componentAppliedLinearAccClamp=<250 250 850>
    componentLinearAccScale=<0.2 0.2 0.6>
    componentLinearVelScale=<0.4 0.4 0.8>
    gravityOverride {z=-980}
    physicsBodyDefinitions [{
      sphereCollisionRadius=0
      boundBone {boneName=dyn_ponytail_04}
      boxExtents=<15 15 15>
      constraintSetup {angularTargetAxis=AxisY, angularLimitsMax=<3 15 25>, angularLimitsMin=<-3 -15 0>, angularTarget=<0 -10 0>}
      localJointOffset=<2.5 0 0>
     }]
   }
   n34 = Anim Dynamics - Bone: dyn_ponytail_05!(ComponentPose=@n33, Alpha=1, GravityOverride:Vector="0,0,-980", ExternalForce="0,0,0") -> (Pose) {
    angularDampingOverride=0.85
    angularSpringConstant=1500
    bOverrideAngularDamping=true
    bUseGravityOverride=true
    gravityScale=0.35
    simulationSpace=RootRelative
    alphaBoolBlend {blendOption=Linear}
    boundBone {boneName=dyn_ponytail_05}
    componentAppliedLinearAccClamp=<250 250 850>
    componentLinearAccScale=<0.2 0.2 0.6>
    componentLinearVelScale=<0.4 0.4 0.8>
    gravityOverride {z=-980}
    physicsBodyDefinitions [{
      sphereCollisionRadius=0
      boundBone {boneName=dyn_ponytail_05}
      boxExtents=<15 15 15>
      constraintSetup {angularTargetAxis=AxisY, angularLimitsMax=<3 15 25>, angularLimitsMin=<-3 -15 0>, angularTarget=<0 -10 0>}
      localJointOffset=<3 0 0>
     }]
   }
   n35 = Anim Dynamics - Bone: dyn_ponytail_06(ComponentPose=@n34, Alpha=1, GravityOverride="0,0,-980", ExternalForce="0,0,0") {
    angularDampingOverride=0.85
    angularSpringConstant=1500
    bOverrideAngularDamping=true
    bUseGravityOverride=true
    gravityScale=0.35
    simulationSpace=RootRelative
    alphaBoolBlend {blendOption=Linear}
    boundBone {boneName=dyn_ponytail_06}
    componentAppliedLinearAccClamp=<250 250 850>
    componentLinearAccScale=<0.2 0.2 0.6>
    componentLinearVelScale=<0.4 0.4 0.8>
    gravityOverride {z=-980}
    physicsBodyDefinitions [{
      sphereCollisionRadius=0
      boundBone {boneName=dyn_ponytail_06}
      boxExtents=<15 15 15>
      constraintSetup {angularTargetAxis=AxisY, angularLimitsMax=<3 15 25>, angularLimitsMin=<-3 -15 0>, angularTarget=<0 -10 0>}
      localJointOffset=<3 0 0>
     }]
   }
   Save cached pose 'BackHairAnimDynamics'(Pose=Component To Local(ComponentPose=@n35)) {cacheName=BackHairAnimDynamics}
   n36 = Control Rig(Source=Use cached pose 'InputPose'(), is carrying=Get bIsDBNOCarrying()) {
    alphaBoolBlend {blendOption=Linear}
    controlRigAssetReference {blueprintRigClass=&F_MED_AkimboEnvoy_FaceAcc_CtrlRig_C}
   }
   Save cached pose 'dbno pose'(Pose=@n36) {cacheName=dbno pose}
  ]}
 EventGraph {nodes [
   Event OnCharacterCustomizationLoaded()
   > Set Skel Mesh(Skel Mesh=GetPartSkeletalMeshComponent())
   n1 = Event BlueprintUpdateAnimation()
   > K2Node_Composite_0 = Gravity!(execute) -> (then:Exec) {is=title K2Node_Composite}
   > K2Node_Composite_1 = Variables!(execute, DeltaTime:double=@n1.DeltaTimeX) -> (then:Exec) {is=title K2Node_Composite}
  ]}
 Gravity {
  path=EventGraph.K2Node_Composite_0.Gravity
  nodes [
   Inputs()
   > n1 = Gravity Override State Function DEPRECATED(
    CharacterPartAnimInstance=Self-Reference()
    GameTimeInSeconds=GetGameTimeInSeconds()
    GravityOverrideParams=Get Gravity Override Params()
    RBAN Params=Get RBAN Params()
   )
   > Set Gravity Override(Gravity Override=@n1)
   > Outputs!(then:Exec)
  ]
 }
 Variables {
  path=EventGraph.K2Node_Composite_1.Variables
  nodes [
   n1 = Inputs!() -> (execute, DeltaTime:double=~)
   n2 = MapRangeClamped(Value=@n1, InRangeA=0.02, InRangeB=0.03, OutRangeB=0.1)
   n3 = Get LowFPSAlpha()
   > Set LowFPSAlpha(LowFPSAlpha:float=SelectFloat(A=0, B=FInterpTo(Current=@n3, Target=@n2, DeltaTime=@n1, InterpSpeed=1), bPickA=Get bIsSliding())) -> (Output_Get:float)
   n4 = Get Skel Mesh()
   > Set LeftClavicleUpWithParachute(LeftClavicleUpWithParachute:float=SelectFloat(
     A=GetAlphaFromReferenceTransformDelta(SkeletalMeshComponent=@n4, BoneName=clavicle_l, BaseName=spine_05, DeltaComponentType=EulerY, InputRange=(X=20,Y=35), TwistAxis="1,0,0")
     B=0
     bPickA=Get bIsParachuteOpened()
    )) -> (Output_Get:float)
   n5 = CheckForBackbling()
   > Set bIsWearingBackpack(bIsWearingBackpack:bool=OR Boolean(A=@n5, B=@n5.bIsBackblingLong)) -> (Output_Get:bool)
   > Set bIsWearingCape(bIsWearingCape:bool=@n5.bIsBackblingACape) -> (Output_Get:bool)
   > n6 = Sequence()
   > Branch(Condition=Get bIsSkydiveDiveMode())
   > RetriggerableDelay(Duration=0.1)
   > Set bIsExitingDive(bIsExitingDive:bool=true) -> (Output_Get:bool)
   > RetriggerableDelay(Duration=0.5)
   > Set bIsExitingDive(bIsExitingDive=false)
   Set HeadUpAlpha(
    execute=@n6.then_1
    HeadUpAlpha:double=GetAlphaFromReferenceTransformDelta(
     SkeletalMeshComponent=Get Skel Mesh()
     BoneName=head
     BaseName=spine_05
     DeltaComponentType=EulerY
     InputRange=(X=10,Y=35)
     TwistAxis="1,0,0"
    )
   ) -> (Output_Get:double)
   > Outputs()
  ]
 }
]
variables [
 bIsExitingDive {type=boolean}
 bIsWearingBackpack {type=boolean}
 bIsWearingCape {type=boolean}
 Gravity Override {type=Vector}
 Gravity Override Params {
  type=GravityOverrideParamsStruct
  default {skydive_Parachuting {gravityOverride=<0 -400 -980>, windAmplitude=<200 -250 50>, windFrequency=<1.5 3.5 4>}}
 }
 HeadUpAlpha {type=number}
 LeftClavicleUpWithParachute {type=number}
 LowFPSAlpha {type=number}
 RBAN Params {
  type=RBANParamsStruct
  default {
   dBNO {simSpaceSettings {
     maxAngularAcceleration=1000
     maxAngularVelocity=3
     maxLinearAcceleration=300
     maxLinearVelocity=700
     velocityScaleZ=0
     externalLinearDragV=<0 0.5 0>
    }}
   dBNO_Carried {simSpaceSettings {
     maxAngularAcceleration=1000
     maxAngularVelocity=3
     maxLinearAcceleration=300
     maxLinearVelocity=700
     velocityScaleZ=0
     externalLinearDragV=<0 0.5 0>
    }}
   emoteOrMelee {simSpaceSettings {
     maxAngularAcceleration=1000
     maxAngularVelocity=3
     maxLinearAcceleration=300
     maxLinearVelocity=700
     velocityScaleZ=0
     externalLinearDragV=<0 0.5 0>
    }}
   front_End {simSpaceSettings {
     maxAngularAcceleration=1000
     maxAngularVelocity=3
     maxLinearAcceleration=300
     maxLinearVelocity=700
     velocityScaleZ=0
     externalLinearDragV=<0 0.5 0>
    }}
   onGround_Crouch_Moving {simSpaceSettings {
     maxAngularAcceleration=1000
     maxAngularVelocity=3
     maxLinearAcceleration=300
     maxLinearVelocity=700
     velocityScaleZ=0
     externalLinearDragV=<0 0.5 0>
    }}
   onGround_Falling {simSpaceSettings {
     maxAngularAcceleration=1000
     maxAngularVelocity=3
     maxLinearAcceleration=300
     maxLinearVelocity=700
     velocityScaleZ=0
     externalLinearDragV=<0 0.5 0>
    }}
   onGround_Moving {simSpaceSettings {
     maxAngularAcceleration=1000
     maxAngularVelocity=3
     maxLinearAcceleration=300
     maxLinearVelocity=700
     velocityScaleZ=0
     externalLinearDragV=<0 0.5 0>
    }}
   onGround_Standing {simSpaceSettings {
     maxAngularAcceleration=1000
     maxAngularVelocity=3
     maxLinearAcceleration=300
     maxLinearVelocity=700
     velocityScaleZ=0
     externalLinearDragV=<0 0.5 0>
    }}
   rail_Grinding {simSpaceSettings {
     maxAngularAcceleration=1000
     maxAngularVelocity=3
     maxLinearAcceleration=300
     maxLinearVelocity=700
     velocityScaleZ=0
     worldAlpha=0.1
     externalLinearDragV=<0 0.5 0>
    }}
   riding_Moving {simSpaceSettings {
     maxAngularAcceleration=1000
     maxAngularVelocity=3
     maxLinearAcceleration=300
     maxLinearVelocity=700
     velocityScaleZ=0
     externalLinearDragV=<0 0.5 0>
    }}
   riding_Moving_Targeting {simSpaceSettings {
     maxAngularAcceleration=1000
     maxAngularVelocity=3
     maxLinearAcceleration=300
     maxLinearVelocity=700
     velocityScaleZ=0
     externalLinearDragV=<0 0.5 0>
    }}
   riding_Standing {simSpaceSettings {
     maxAngularAcceleration=1000
     maxAngularVelocity=3
     maxLinearAcceleration=300
     maxLinearVelocity=700
     velocityScaleZ=0
     externalLinearDragV=<0 0.5 0>
    }}
   skydive_DivingDown {simSpaceSettings {
     maxAngularAcceleration=1000
     maxAngularVelocity=3
     maxLinearAcceleration=300
     maxLinearVelocity=700
     velocityScaleZ=0
     externalLinearDragV=<0 0.5 0>
    }}
   skydive_FloatingDown {
    componentLinearAccScale=<0 0.5 0>
    componentLinearVelScale=<0 0.5 0>
    simSpaceSettings {
     maxAngularAcceleration=1000
     maxAngularVelocity=3
     maxLinearAcceleration=300
     maxLinearVelocity=700
     velocityScaleZ=0
     externalLinearDragV=<0 0.5 0>
    }
   }
   skydive_GoingUp {simSpaceSettings {
     maxAngularAcceleration=1000
     maxAngularVelocity=3
     maxLinearAcceleration=300
     maxLinearVelocity=700
     velocityScaleZ=0
     externalLinearDragV=<0 0.5 0>
    }}
   skydive_Parachuting {
    componentLinearAccScale=<0 0.1 0.3>
    componentLinearVelScale=<0 0.1 0.3>
    simSpaceSettings {
     maxAngularAcceleration=1000
     maxAngularVelocity=3
     maxLinearAcceleration=300
     maxLinearVelocity=700
     velocityScaleZ=0
     externalLinearDragV=<0 0.5 0>
    }
   }
   skydive_Surfing {simSpaceSettings {
     maxAngularAcceleration=1000
     maxAngularVelocity=3
     maxLinearAcceleration=300
     maxLinearVelocity=700
     velocityScaleZ=0
     externalLinearDragV=<0 0.5 0>
    }}
   swimming {simSpaceSettings {
     maxAngularAcceleration=1000
     maxAngularVelocity=3
     maxLinearAcceleration=300
     maxLinearVelocity=700
     velocityScaleZ=0
     externalLinearDragV=<0 0.5 0>
    }}
   tactical_Sprint {simSpaceSettings {
     maxAngularAcceleration=1000
     maxAngularVelocity=3
     maxLinearAcceleration=300
     maxLinearVelocity=700
     velocityScaleZ=0
     externalLinearDragV=<0 0.5 0>
    }}
   ziplining {simSpaceSettings {
     maxAngularAcceleration=1000
     maxAngularVelocity=3
     maxLinearAcceleration=300
     maxLinearVelocity=700
     velocityScaleZ=0
     externalLinearDragV=<0 0.5 0>
    }}
  }
 }
 Skel Mesh {type=SkeletalMeshComponent}
]
